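# Sound cue for Burt's charge-attack vocalization, using prey vocalization attenuation and the tactical creature vocalization sound class.
# A random gate with replacement plays a vocal about one time in three, otherwise silence, under pitch and volume modulation.
# A quality switch feeds the first two levels from eight random takes and the last two from three.
# Concurrency combines footstep local and global limits with an owner-limited VO set.
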SoundCue /Irwin/Sounds/Burt/Vocalizations/Burt_Vocal_ChargeAttack_Cue
root=Modulator_0
nodes [
 {id=Modulator_0, children [Random_6], properties {pitchMax=1.1, volumeMax=1.1, volumeMin=0.9}}
 {id=QualityLevel_0, children [Random_3,Random_3,Random_5,Random_5]}
 {
  id=Random_3
  children [WavePlayer_0, WavePlayer_5, WavePlayer_6, WavePlayer_7, WavePlayer_8, WavePlayer_9, WavePlayer_10, WavePlayer_11]
  properties {weights [1,1,1,1,1,1,1,1]}
 }
 {id=Random_5, children [WavePlayer_16, WavePlayer_17, WavePlayer_18], properties {bShouldExcludeFromBranchCulling=true, weights [1,1,1]}}
 {
  id=Random_6
  children [QualityLevel_0, null]
  properties {bRandomizeWithoutReplacement=false, bShouldExcludeFromBranchCulling=true, weights [0.5,1]}
 }
 {id=WavePlayer_0, properties {soundWaveAssetPtr=&/Burt_Vocal_ChargeAttack_01}}
 {id=WavePlayer_10, properties {soundWaveAssetPtr=&/Burt_Vocal_ChargeAttack_07}}
 {id=WavePlayer_11, properties {soundWaveAssetPtr=&/Burt_Vocal_ChargeAttack_08}}
 {id=WavePlayer_16, properties {soundWaveAssetPtr=&Burt_Vocal_ChargeAttack_01}}
 {id=WavePlayer_17, properties {soundWaveAssetPtr=&/Burt_Vocal_ChargeAttack_02}}
 {id=WavePlayer_18, properties {soundWaveAssetPtr=&/Burt_Vocal_ChargeAttack_03}}
 {id=WavePlayer_5, properties {soundWaveAssetPtr=&Burt_Vocal_ChargeAttack_02}}
 {id=WavePlayer_6, properties {soundWaveAssetPtr=&Burt_Vocal_ChargeAttack_03}}
 {id=WavePlayer_7, properties {soundWaveAssetPtr=&/Burt_Vocal_ChargeAttack_04}}
 {id=WavePlayer_8, properties {soundWaveAssetPtr=&/Burt_Vocal_ChargeAttack_05}}
 {id=WavePlayer_9, properties {soundWaveAssetPtr=&/Burt_Vocal_ChargeAttack_06}}
]
properties {
 attenuationSettings=/Irwin/Sounds/Settings/Prey_Vocalization_Att
 maxDistance=6350
 priority=0.7
 soundClassObject=&Fort_Creatures_Vocalization_Tactical
 volumeMultiplier=0.7
 concurrencySet [
  &Athena_Footstep_Local_Conc
  &Athena_Footstep_Global_Conc
  /Irwin/Sounds/Settings/Conc/SetOneLimitToOwner_VO_Conc
 ]
}
