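# Sound cue for the charge vocalization of Burt, an Irwin creature, varying by listener relationship.
# IsLocal plays either random chase breathing or charge-attack waves; otherwise it plays random agitated waves, routed by IsRiddenByEnemy.
# Enemy-ridden uses enemy-foley class and riding attenuation at reduced volume; not enemy-ridden is quieter still, and unset is unreduced, both using creature class and howl attenuation.
# Slight pitch and volume modulation adds variation, and concurrency rules limit overlapping plays.
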SoundCue /Irwin/Sounds/Burt/Vocalizations/Burt_Vocal_Charge_Cue
root=Branch_2
nodes [
 {id=Attenuation_0, children [Modulator_1], properties {attenuationSettings=/Irwin/Sounds/Settings/Predator_Vocalization_Howl_Close_Att}}
 {id=Attenuation_2, children [Modulator_1], properties {attenuationSettings=/Game/Sounds/Attenuation/Fort_Riding_Enemy_VO_Att}}
 {id=Attenuation_3, children [Modulator_1], properties {attenuationSettings=&Predator_Vocalization_Howl_Close_Att}}
 {id=Attenuation_4, children [Modulator_0], properties {attenuationSettings=&Predator_Vocalization_Howl_Close_Att}}
 {id=Branch_1, children [Modulator_3,Modulator_2,SoundClass_0], properties {boolParameterName=IsRiddenByEnemy}}
 {id=Branch_2, children [SoundClass_4,Branch_1,Branch_1], properties {boolParameterName=IsLocal}}
 {id=Mixer_1, children [Random_6], properties {inputVolume [1.25]}}
 {id=Mixer_2, children [Random_7], properties {inputVolume [1]}}
 {id=Modulator_0, children [Random_8], properties {pitchMax=1.1, pitchMin=0.9, volumeMax=1, volumeMin=0.9}}
 {id=Modulator_1, children [Random_0], properties {pitchMax=1.1, pitchMin=0.9, volumeMax=1, volumeMin=0.9}}
 {id=Modulator_2, children [SoundClass_3], properties {pitchMax=1, pitchMin=1, volumeMax=0.3, volumeMin=0.3}}
 {id=Modulator_3, children [SoundClass_2], properties {pitchMax=1, pitchMin=1, volumeMax=0.7, volumeMin=0.7}}
 {
  id=Random_0
  children [WavePlayer_1,WavePlayer_2,WavePlayer_3,WavePlayer_4]
  properties {bShouldExcludeFromBranchCulling=true, weights [1,1,1,1]}
 }
 {
  id=Random_6
  children [WavePlayer_14, WavePlayer_15, WavePlayer_16, WavePlayer_17, WavePlayer_18, WavePlayer_21]
  properties {bShouldExcludeFromBranchCulling=true, weights [1,1,1,1,1,1]}
 }
 {
  id=Random_7
  children [WavePlayer_19, WavePlayer_20, WavePlayer_25, WavePlayer_26, WavePlayer_36, WavePlayer_40, WavePlayer_41, WavePlayer_42]
  properties {bShouldExcludeFromBranchCulling=true, weights [1,1,1,1,1,1,1,1]}
 }
 {id=Random_8, children [Mixer_1,Mixer_2], properties {weights [1,1]}}
 {
  id=SoundClass_0
  children [Attenuation_0]
  properties {soundClassOverride=&Fort_Creatures_Vocalization_Tactical}
 }
 {
  id=SoundClass_2
  children [Attenuation_2]
  properties {soundClassOverride=&Fort_Player_Foley_3P_Enemy_Athena}
 }
 {
  id=SoundClass_3
  children [Attenuation_3]
  properties {soundClassOverride=&Fort_Creatures_Vocalization_Tactical}
 }
 {
  id=SoundClass_4
  children [Attenuation_4]
  properties {soundClassOverride=&Fort_Creatures_Vocalization_Tactical}
 }
 {id=WavePlayer_1, properties {soundWaveAssetPtr=&/Burt_Vocal_Agitated_01}}
 {id=WavePlayer_14, properties {soundWaveAssetPtr=&/Burt_Vocal_Chase_Breathing_01}}
 {id=WavePlayer_15, properties {soundWaveAssetPtr=&/Burt_Vocal_Chase_Breathing_09}}
 {id=WavePlayer_16, properties {soundWaveAssetPtr=&/Burt_Vocal_Chase_Breathing_03}}
 {id=WavePlayer_17, properties {soundWaveAssetPtr=&/Burt_Vocal_Chase_Breathing_04}}
 {id=WavePlayer_18, properties {soundWaveAssetPtr=&/Burt_Vocal_Chase_Breathing_05}}
 {id=WavePlayer_19, properties {soundWaveAssetPtr=&/Burt_Vocal_ChargeAttack_01}}
 {id=WavePlayer_2, properties {soundWaveAssetPtr=&/Burt_Vocal_Agitated_02}}
 {id=WavePlayer_20, properties {soundWaveAssetPtr=&/Burt_Vocal_ChargeAttack_02}}
 {id=WavePlayer_21, properties {soundWaveAssetPtr=&/Burt_Vocal_Chase_Breathing_06}}
 {id=WavePlayer_25, properties {soundWaveAssetPtr=&/Burt_Vocal_ChargeAttack_03}}
 {id=WavePlayer_26, properties {soundWaveAssetPtr=&/Burt_Vocal_ChargeAttack_04}}
 {id=WavePlayer_3, properties {soundWaveAssetPtr=&/Burt_Vocal_Agitated_06}}
 {id=WavePlayer_36, properties {soundWaveAssetPtr=&/Burt_Vocal_ChargeAttack_05}}
 {id=WavePlayer_4, properties {soundWaveAssetPtr=&/Burt_Vocal_Agitated_07}}
 {id=WavePlayer_40, properties {soundWaveAssetPtr=&/Burt_Vocal_ChargeAttack_06}}
 {id=WavePlayer_41, properties {soundWaveAssetPtr=&/Burt_Vocal_ChargeAttack_07}}
 {id=WavePlayer_42, properties {soundWaveAssetPtr=&/Burt_Vocal_ChargeAttack_08}}
]
properties {
 maxDistance=18000
 volumeMultiplier=1.5
 concurrencyOverrides {bLimitToOwner=true, maxCount=2, resolutionRule=PreventNew, volumeScaleAttackTime=0.2}
 concurrencySet [
  &Athena_Footstep_Local_Conc
  &Athena_Footstep_Global_Conc
  /Game/Sounds/Concurrency_Rules/LimitToOnePreventNew
 ]
}
