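# Sound cue for Burt's chase breathing vocalization, shifting from chase breaths to tiring breaths as fatigue rises.
# IsLocal and IsRiddenByEnemy branches set volume, sound class and attenuation, with separate enemy-rider treatment, before converging on one pitch- and volume-randomized chain.
# A quality-level node selects between two wave sets, each crossfading three FatigueLevel bands.
# The top band randomly mixes tiring breaths, agitated calls and silence.
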SoundCue /Irwin/Sounds/Burt/Vocalizations/Burt_Vocal_Chase_Breathing_Cue
root=Branch_2
nodes [
 {id=Attenuation_0, children [Modulator_1], properties {attenuationSettings=/Irwin/Sounds/Settings/Prey_Vocalization_Att}}
 {id=Attenuation_2, children [Modulator_1], properties {attenuationSettings=/Game/Sounds/Attenuation/Fort_Riding_Enemy_VO_Att}}
 {id=Attenuation_3, children [Modulator_1], properties {attenuationSettings=&Prey_Vocalization_Att}}
 {id=Attenuation_4, children [Modulator_1], properties {attenuationSettings=&Prey_Vocalization_Att}}
 {id=Branch_1, children [Modulator_3,Modulator_2,SoundClass_0], properties {boolParameterName=IsRiddenByEnemy}}
 {id=Branch_2, children [Modulator_0,Branch_1,Branch_1], properties {boolParameterName=IsLocal}}
 {id=Modulator_0, children [SoundClass_4], properties {pitchMax=1, pitchMin=1, volumeMax=0.9, volumeMin=0.7}}
 {id=Modulator_1, children [QualityLevel_0], properties {pitchMax=1.1, pitchMin=0.9, volumeMax=1, volumeMin=0.5}}
 {id=Modulator_2, children [SoundClass_3], properties {pitchMax=1, pitchMin=1, volumeMax=0.3, volumeMin=0.3}}
 {id=Modulator_3, children [SoundClass_2], properties {pitchMax=1, pitchMin=1, volumeMax=0.7, volumeMin=0.7}}
 {id=Modulator_4, children [Random_6], properties {pitchMax=1, pitchMin=1, volumeMax=2.2, volumeMin=2}}
 {id=Modulator_5, children [Random_14], properties {pitchMax=1, pitchMin=1, volumeMax=2.2, volumeMin=2}}
 {
  id=ParamCrossFade_1
  children [Random_2,Random_17,Random_16]
  properties {
   paramName=FatigueLevel
   crossFadeInput [
    {fadeInDistanceEnd=0, fadeInDistanceStart=0, fadeOutDistanceEnd=0.85, fadeOutDistanceStart=0.82, volume=1.8}
    {fadeInDistanceEnd=0.85, fadeInDistanceStart=0.82, fadeOutDistanceEnd=0.96, fadeOutDistanceStart=0.95, volume=2}
    {fadeInDistanceEnd=0.96, fadeInDistanceStart=0.95, fadeOutDistanceEnd=1, fadeOutDistanceStart=1, volume=2}
   ]
  }
 }
 {
  id=ParamCrossFade_2
  children [Random_1,Random_0,Random_13]
  properties {
   paramName=FatigueLevel
   crossFadeInput [
    {fadeInDistanceEnd=0, fadeInDistanceStart=0, fadeOutDistanceEnd=0.85, fadeOutDistanceStart=0.82, volume=1.8}
    {fadeInDistanceEnd=0.85, fadeInDistanceStart=0.82, fadeOutDistanceEnd=0.96, fadeOutDistanceStart=0.95, volume=2}
    {fadeInDistanceEnd=0.96, fadeInDistanceStart=0.95, fadeOutDistanceEnd=1, fadeOutDistanceStart=1, volume=2}
   ]
  }
 }
 {id=QualityLevel_0, children [ParamCrossFade_2, ParamCrossFade_2, ParamCrossFade_1, ParamCrossFade_1]}
 {
  id=Random_0
  children [WavePlayer_1, WavePlayer_2, WavePlayer_3, WavePlayer_4, WavePlayer_23, WavePlayer_24, null, null, null, null, null]
  properties {weights [1,1,1,1,1,1,1,1,1,1,1]}
 }
 {
  id=Random_1
  children [WavePlayer_0, WavePlayer_7, WavePlayer_8, WavePlayer_9, WavePlayer_10, WavePlayer_11, null, null, null, null, null]
  properties {weights [1,1,1,1,1,1,1,1,1,1,1]}
 }
 {id=Random_12, children [WavePlayer_32, WavePlayer_36, WavePlayer_40, WavePlayer_44], properties {weights [1,1,1,1]}}
 {id=Random_13, children [Modulator_4,Random_12,null], properties {bRandomizeWithoutReplacement=false, weights [1,0.8,0.7]}}
 {id=Random_14, children [WavePlayer_54, WavePlayer_55, WavePlayer_56, WavePlayer_57], properties {weights [1,1,1,1]}}
 {id=Random_15, children [WavePlayer_50, WavePlayer_51, WavePlayer_52, WavePlayer_53], properties {weights [1,1,1,1]}}
 {id=Random_16, children [Modulator_5,Random_15,null], properties {bRandomizeWithoutReplacement=false, weights [1,0.8,0.7]}}
 {
  id=Random_17
  children [WavePlayer_58, WavePlayer_59, WavePlayer_60, WavePlayer_61, null, null, null, null, null]
  properties {weights [1,1,1,1,1,1,1,1,1]}
 }
 {
  id=Random_2
  children [WavePlayer_19, WavePlayer_20, WavePlayer_21, WavePlayer_22]
  properties {bShouldExcludeFromBranchCulling=true, weights [1,1,1,1]}
 }
 {id=Random_6, children [WavePlayer_45, WavePlayer_46, WavePlayer_47, WavePlayer_48, WavePlayer_49], properties {weights [1,1,1,1,1]}}
 {
  id=SoundClass_0
  children [Attenuation_0]
  properties {soundClassOverride=&Fort_Creatures_Vocalization_Tactical}
 }
 {
  id=SoundClass_2
  children [Attenuation_2]
  properties {soundClassOverride=&Fort_Player_Foley_3P_Enemy_Athena}
 }
 {
  id=SoundClass_3
  children [Attenuation_3]
  properties {soundClassOverride=&Fort_Creatures_Vocalization_Tactical}
 }
 {
  id=SoundClass_4
  children [Attenuation_4]
  properties {soundClassOverride=&Fort_Creatures_Vocalization_Tactical}
 }
 {id=WavePlayer_0, properties {soundWaveAssetPtr=&/Burt_Vocal_Chase_Breathing_01}}
 {id=WavePlayer_1, properties {soundWaveAssetPtr=&/Burt_Vocal_Sprint_Tiring_Breathing_01}}
 {id=WavePlayer_10, properties {soundWaveAssetPtr=&/Burt_Vocal_Chase_Breathing_05}}
 {id=WavePlayer_11, properties {soundWaveAssetPtr=&/Burt_Vocal_Chase_Breathing_06}}
 {id=WavePlayer_19, properties {soundWaveAssetPtr=&Burt_Vocal_Chase_Breathing_01}}
 {id=WavePlayer_2, properties {soundWaveAssetPtr=&/Burt_Vocal_Sprint_Tiring_Breathing_02}}
 {id=WavePlayer_20, properties {soundWaveAssetPtr=&/Burt_Vocal_Chase_Breathing_09}}
 {id=WavePlayer_21, properties {soundWaveAssetPtr=&/Burt_Vocal_Chase_Breathing_03}}
 {id=WavePlayer_22, properties {soundWaveAssetPtr=&/Burt_Vocal_Chase_Breathing_04}}
 {id=WavePlayer_23, properties {soundWaveAssetPtr=&/Burt_Vocal_Sprint_Tiring_Breathing_06}}
 {id=WavePlayer_24, properties {soundWaveAssetPtr=&/Burt_Vocal_Sprint_Tiring_Breathing_07}}
 {id=WavePlayer_3, properties {soundWaveAssetPtr=&/Burt_Vocal_Sprint_Tiring_Breathing_03}}
 {id=WavePlayer_32, properties {soundWaveAssetPtr=&/Burt_Vocal_Agitated_01}}
 {id=WavePlayer_36, properties {soundWaveAssetPtr=&/Burt_Vocal_Agitated_02}}
 {id=WavePlayer_4, properties {soundWaveAssetPtr=&/Burt_Vocal_Sprint_Tiring_Breathing_05}}
 {id=WavePlayer_40, properties {soundWaveAssetPtr=&/Burt_Vocal_Agitated_06}}
 {id=WavePlayer_44, properties {soundWaveAssetPtr=&/Burt_Vocal_Agitated_07}}
 {id=WavePlayer_45, properties {soundWaveAssetPtr=&Burt_Vocal_Sprint_Tiring_Breathing_01}}
 {id=WavePlayer_46, properties {soundWaveAssetPtr=&Burt_Vocal_Sprint_Tiring_Breathing_02}}
 {id=WavePlayer_47, properties {soundWaveAssetPtr=&Burt_Vocal_Sprint_Tiring_Breathing_03}}
 {id=WavePlayer_48, properties {soundWaveAssetPtr=&Burt_Vocal_Sprint_Tiring_Breathing_05}}
 {id=WavePlayer_49, properties {soundWaveAssetPtr=&Burt_Vocal_Sprint_Tiring_Breathing_06}}
 {id=WavePlayer_50, properties {soundWaveAssetPtr=&Burt_Vocal_Agitated_01}}
 {id=WavePlayer_51, properties {soundWaveAssetPtr=&Burt_Vocal_Agitated_02}}
 {id=WavePlayer_52, properties {soundWaveAssetPtr=&Burt_Vocal_Agitated_06}}
 {id=WavePlayer_53, properties {soundWaveAssetPtr=&Burt_Vocal_Agitated_07}}
 {id=WavePlayer_54, properties {soundWaveAssetPtr=&Burt_Vocal_Sprint_Tiring_Breathing_01}}
 {id=WavePlayer_55, properties {soundWaveAssetPtr=&Burt_Vocal_Sprint_Tiring_Breathing_02}}
 {id=WavePlayer_56, properties {soundWaveAssetPtr=&Burt_Vocal_Sprint_Tiring_Breathing_03}}
 {id=WavePlayer_57, properties {soundWaveAssetPtr=&Burt_Vocal_Sprint_Tiring_Breathing_05}}
 {id=WavePlayer_58, properties {soundWaveAssetPtr=&Burt_Vocal_Sprint_Tiring_Breathing_01}}
 {id=WavePlayer_59, properties {soundWaveAssetPtr=&Burt_Vocal_Sprint_Tiring_Breathing_02}}
 {id=WavePlayer_60, properties {soundWaveAssetPtr=&Burt_Vocal_Sprint_Tiring_Breathing_03}}
 {id=WavePlayer_61, properties {soundWaveAssetPtr=&Burt_Vocal_Sprint_Tiring_Breathing_05}}
 {id=WavePlayer_7, properties {soundWaveAssetPtr=&Burt_Vocal_Chase_Breathing_09}}
 {id=WavePlayer_8, properties {soundWaveAssetPtr=&Burt_Vocal_Chase_Breathing_03}}
 {id=WavePlayer_9, properties {soundWaveAssetPtr=&Burt_Vocal_Chase_Breathing_04}}
]
properties {
 maxDistance=9000
 volumeMultiplier=0.6
 concurrencySet [
  /Game/Sounds/Concurrency_Rules/Burt_Vocal_Conc
  &Athena_Footstep_Global_Conc
  /Game/Sounds/Concurrency_Rules/LimitToOnePreventNewQuick
 ]
}
