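# Sound cue for Burt creature chase vocalizations: aggressive, breathing, grunting and snorting sounds.
# IsLocal branches to a tactical creature sound class; otherwise IsRiddenByEnemy picks enemy-rider class and attenuation, a quieter prey route, or an unscaled fallback when unset.
# All routes share pitch and volume randomization, then a quality-level switch between nine-wave and four-wave random pools.
# Concurrency reuses the Athena local and global footstep sets.
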SoundCue /Irwin/Sounds/Burt/Vocalizations/Burt_Vocal_Chase_Cue
root=Branch_2
nodes [
 {id=Attenuation_0, children [Modulator_1], properties {attenuationSettings=/Irwin/Sounds/Settings/Prey_Vocalization_Att}}
 {id=Attenuation_2, children [Modulator_1], properties {attenuationSettings=/Game/Sounds/Attenuation/Fort_Riding_Enemy_VO_Att}}
 {id=Attenuation_3, children [Modulator_1], properties {attenuationSettings=&Prey_Vocalization_Att}}
 {id=Attenuation_4, children [Modulator_1], properties {attenuationSettings=&Prey_Vocalization_Att}}
 {id=Branch_1, children [Modulator_3,Modulator_2,SoundClass_0], properties {boolParameterName=IsRiddenByEnemy}}
 {id=Branch_2, children [SoundClass_4,Branch_1,Branch_1], properties {boolParameterName=IsLocal}}
 {id=Modulator_1, children [QualityLevel_0], properties {pitchMax=1.1, pitchMin=0.9, volumeMax=1, volumeMin=0.5}}
 {id=Modulator_2, children [SoundClass_3], properties {pitchMax=1, pitchMin=1, volumeMax=0.3, volumeMin=0.3}}
 {id=Modulator_3, children [SoundClass_2], properties {pitchMax=1, pitchMin=1, volumeMax=0.7, volumeMin=0.7}}
 {id=QualityLevel_0, children [Random_2,Random_2,Random_0,Random_0]}
 {
  id=Random_0
  children [WavePlayer_0,WavePlayer_1,WavePlayer_2,WavePlayer_3]
  properties {bShouldExcludeFromBranchCulling=true, weights [1,1,1,1]}
 }
 {
  id=Random_2
  children [WavePlayer_37, WavePlayer_43, WavePlayer_47, WavePlayer_49, WavePlayer_50, WavePlayer_51, WavePlayer_52, WavePlayer_53, WavePlayer_55]
  properties {weights [1,1,1,1,1,1,1,1,1]}
 }
 {
  id=SoundClass_0
  children [Attenuation_0]
  properties {soundClassOverride=&Fort_Creatures_Vocalization_Tactical}
 }
 {
  id=SoundClass_2
  children [Attenuation_2]
  properties {soundClassOverride=&Fort_Player_Foley_3P_Enemy_Athena}
 }
 {
  id=SoundClass_3
  children [Attenuation_3]
  properties {soundClassOverride=&Fort_Creatures_Vocalization_Tactical}
 }
 {
  id=SoundClass_4
  children [Attenuation_4]
  properties {soundClassOverride=&Fort_Creatures_Vocalization_Tactical}
 }
 {id=WavePlayer_0, properties {soundWaveAssetPtr=&/Burt_Vocal_Chase_Agressive_04}}
 {id=WavePlayer_1, properties {soundWaveAssetPtr=&/Burt_Vocal_Chase_Breathing_07}}
 {id=WavePlayer_2, properties {soundWaveAssetPtr=&/Burt_Vocal_Chase_Grunting_01}}
 {id=WavePlayer_3, properties {soundWaveAssetPtr=&/Burt_Vocal_Chase_Snorting_Long_02}}
 {id=WavePlayer_37, properties {soundWaveAssetPtr=&Burt_Vocal_Chase_Agressive_04}}
 {id=WavePlayer_43, properties {soundWaveAssetPtr=&Burt_Vocal_Chase_Breathing_07}}
 {id=WavePlayer_47, properties {soundWaveAssetPtr=&Burt_Vocal_Chase_Grunting_01}}
 {id=WavePlayer_49, properties {soundWaveAssetPtr=&Burt_Vocal_Chase_Snorting_Long_02}}
 {id=WavePlayer_50, properties {soundWaveAssetPtr=&/Burt_Vocal_Chase_Snorting_Long_04}}
 {id=WavePlayer_51, properties {soundWaveAssetPtr=&/Burt_Vocal_Chase_Snorting_Long_05}}
 {id=WavePlayer_52, properties {soundWaveAssetPtr=&/Burt_Vocal_Chase_Snorting_Short_02}}
 {id=WavePlayer_53, properties {soundWaveAssetPtr=&/Burt_Vocal_Chase_Snorting_Short_01}}
 {id=WavePlayer_55, properties {soundWaveAssetPtr=&/Burt_Vocal_Chase_Snorting_Short_04}}
]
properties {
 maxDistance=9000
 pitchMultiplier=0.9
 volumeMultiplier=1.6
 concurrencySet [&Athena_Footstep_Local_Conc, &Athena_Footstep_Global_Conc]
}
