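# Sound cue for Burt's death vocalization, playing short death cries with slight random pitch and volume variation.
# A quality-level switch feeds the modulator, sending its first two levels to a three-wave random pick and its last two to another.
# The second pick has two waves plus an empty slot and is excluded from branch culling.
# Prey vocalization attenuation, creature vocalization class, footstep and per-owner VO concurrency govern playback.
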SoundCue /Irwin/Sounds/Burt/Vocalizations/Burt_Vocal_Death_Cue
root=Modulator_0
nodes [
 {id=Modulator_0, children [QualityLevel_0], properties {pitchMax=1.1, pitchMin=1, volumeMax=1.2, volumeMin=1.1}}
 {id=QualityLevel_0, children [Random_0,Random_0,Random_1,Random_1]}
 {id=Random_0, children [WavePlayer_2,WavePlayer_6,WavePlayer_7], properties {weights [1,1,1]}}
 {id=Random_1, children [WavePlayer_0,WavePlayer_1,null], properties {bShouldExcludeFromBranchCulling=true, weights [1,1,1]}}
 {id=WavePlayer_0, properties {soundWaveAssetPtr=&/Burt_Vocal_DeathShort_02}}
 {id=WavePlayer_1, properties {soundWaveAssetPtr=&/Burt_Vocal_DeathShort_03}}
 {id=WavePlayer_2, properties {soundWaveAssetPtr=&Burt_Vocal_DeathShort_02}}
 {id=WavePlayer_6, properties {soundWaveAssetPtr=&Burt_Vocal_DeathShort_03}}
 {id=WavePlayer_7, properties {soundWaveAssetPtr=&/Burt_Vocal_DeathShort_04}}
]
properties {
 attenuationSettings=/Irwin/Sounds/Settings/Prey_Vocalization_Att
 maxDistance=6350
 soundClassObject=/Game/Sounds/FortSoundClassesAndModes/Fort_Creatures_Vocalization_Normal
 volumeMultiplier=2.4
 concurrencySet [
  &Athena_Footstep_Local_Conc
  &Athena_Footstep_Global_Conc
  /Irwin/Sounds/Settings/Conc/SetOneLimitToOwner_VO_Conc
 ]
}
