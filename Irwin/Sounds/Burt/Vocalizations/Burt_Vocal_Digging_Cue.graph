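# Sound cue for Burt's digging vocalization, using prey vocalization attenuation and footstep concurrency limits.
# A quality switch sends its first two slots to a mix of eight random idle snorts and pitched-down random digging foley.
# Its last two slots instead layer two random picks from smaller four-snort pools, without foley.
# A root modulator varies pitch and volume over the result, which the cue also boosts in both.
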SoundCue /Irwin/Sounds/Burt/Vocalizations/Burt_Vocal_Digging_Cue
root=Modulator_0
nodes [
 {id=Mixer_1, children [Random_1,Modulator_1], properties {inputVolume [0.8,1.1]}}
 {id=Mixer_3, children [Random_0,Random_3], properties {inputVolume [0.8,1.1]}}
 {id=Modulator_0, children [QualityLevel_0], properties {pitchMax=1.1, volumeMax=1.1, volumeMin=0.9}}
 {id=Modulator_1, children [Random_2], properties {pitchMax=0.9, pitchMin=0.8}}
 {id=QualityLevel_0, children [Mixer_1,Mixer_1,Mixer_3,Mixer_3]}
 {
  id=Random_0
  children [WavePlayer_0,WavePlayer_1,WavePlayer_2,WavePlayer_3]
  properties {bShouldExcludeFromBranchCulling=true, weights [1,1,1,1]}
 }
 {
  id=Random_1
  children [WavePlayer_7, WavePlayer_8, WavePlayer_12, WavePlayer_13, WavePlayer_14, WavePlayer_15, WavePlayer_16, WavePlayer_17]
  properties {weights [1,1,1,1,1,1,1,1]}
 }
 {
  id=Random_2
  children [WavePlayer_4, WavePlayer_5, WavePlayer_6, WavePlayer_9, WavePlayer_10, WavePlayer_11, WavePlayer_18]
  properties {weights [1,1,1,1,1,1,1]}
 }
 {id=Random_3, children [WavePlayer_19, WavePlayer_20, WavePlayer_21, WavePlayer_22], properties {weights [1,1,1,1]}}
 {id=WavePlayer_0, properties {soundWaveAssetPtr=&/Burt_Vocal_Idle_Snorting_01}}
 {id=WavePlayer_1, properties {soundWaveAssetPtr=&/Burt_Vocal_Idle_Snorting_02}}
 {id=WavePlayer_10, properties {soundWaveAssetPtr=/Irwin/Sounds/Burt/Foley/Burt_Foley_Digging_05}}
 {id=WavePlayer_11, properties {soundWaveAssetPtr=/Irwin/Sounds/Burt/Foley/Burt_Foley_Digging_06}}
 {id=WavePlayer_12, properties {soundWaveAssetPtr=&/Burt_Vocal_Idle_Snorting_03}}
 {id=WavePlayer_13, properties {soundWaveAssetPtr=&/Burt_Vocal_Idle_Snorting_04}}
 {id=WavePlayer_14, properties {soundWaveAssetPtr=&/Burt_Vocal_Idle_Snorting_05}}
 {id=WavePlayer_15, properties {soundWaveAssetPtr=&/Burt_Vocal_Idle_Snorting_06}}
 {id=WavePlayer_16, properties {soundWaveAssetPtr=&/Burt_Vocal_Idle_Snorting_07}}
 {id=WavePlayer_17, properties {soundWaveAssetPtr=&/Burt_Vocal_Idle_Snorting_08}}
 {id=WavePlayer_18, properties {soundWaveAssetPtr=/Irwin/Sounds/Burt/Foley/Burt_Foley_Digging_07}}
 {id=WavePlayer_19, properties {soundWaveAssetPtr=&Burt_Vocal_Idle_Snorting_01}}
 {id=WavePlayer_2, properties {soundWaveAssetPtr=&Burt_Vocal_Idle_Snorting_03}}
 {id=WavePlayer_20, properties {soundWaveAssetPtr=&Burt_Vocal_Idle_Snorting_02}}
 {id=WavePlayer_21, properties {soundWaveAssetPtr=&Burt_Vocal_Idle_Snorting_03}}
 {id=WavePlayer_22, properties {soundWaveAssetPtr=&Burt_Vocal_Idle_Snorting_04}}
 {id=WavePlayer_3, properties {soundWaveAssetPtr=&Burt_Vocal_Idle_Snorting_04}}
 {id=WavePlayer_4, properties {soundWaveAssetPtr=/Irwin/Sounds/Burt/Foley/Burt_Foley_Digging_01}}
 {id=WavePlayer_5, properties {soundWaveAssetPtr=/Irwin/Sounds/Burt/Foley/Burt_Foley_Digging_02}}
 {id=WavePlayer_6, properties {soundWaveAssetPtr=/Irwin/Sounds/Burt/Foley/Burt_Foley_Digging_03}}
 {id=WavePlayer_7, properties {soundWaveAssetPtr=&Burt_Vocal_Idle_Snorting_01}}
 {id=WavePlayer_8, properties {soundWaveAssetPtr=&Burt_Vocal_Idle_Snorting_02}}
 {id=WavePlayer_9, properties {soundWaveAssetPtr=/Irwin/Sounds/Burt/Foley/Burt_Foley_Digging_04}}
]
properties {
 attenuationSettings=/Irwin/Sounds/Settings/Prey_Vocalization_Att
 maxDistance=6350
 pitchMultiplier=1.2
 soundClassObject=&Fort_Creatures_Vocalization_Tactical
 volumeMultiplier=1.4
 concurrencySet [&Athena_Footstep_Local_Conc, &Athena_Footstep_Global_Conc]
}
