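# Sound Cue for Burt's eating vocalizations, using the creature vocalization sound class and prey vocalization attenuation.
# A quality switch sends its first two levels to a random pool of sixteen heavy and light takes, and its last two to a reduced pool of four.
# A modulator above the switch randomizes pitch and volume per play.
# Concurrency borrows the local and global Athena footstep sets rather than vocalization-specific ones.
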SoundCue /Irwin/Sounds/Burt/Vocalizations/Burt_Vocal_Eat_Cue
root=Modulator_0
nodes [
 {id=Modulator_0, children [QualityLevel_0], properties {pitchMax=1.15, volumeMax=1.1, volumeMin=0.9}}
 {id=QualityLevel_0, children [Random_1,Random_1,Random_2,Random_2]}
 {
  id=Random_1
  children [
   WavePlayer_1
   WavePlayer_2
   WavePlayer_3
   WavePlayer_4
   WavePlayer_12
   WavePlayer_13
   WavePlayer_14
   WavePlayer_18
   WavePlayer_20
   WavePlayer_21
   WavePlayer_22
   WavePlayer_23
   WavePlayer_24
   WavePlayer_25
   WavePlayer_26
   WavePlayer_27
  ]
  properties {weights [1,1,1,1,1,1,1,1,1,1,1,1,1,1,1,1]}
 }
 {
  id=Random_2
  children [WavePlayer_0, WavePlayer_5, WavePlayer_30, WavePlayer_31]
  properties {bShouldExcludeFromBranchCulling=true, weights [1,1,1,1]}
 }
 {id=WavePlayer_0, properties {soundWaveAssetPtr=&/Burt_Vocal_Eat_Heavy_01}}
 {id=WavePlayer_1, properties {soundWaveAssetPtr=&Burt_Vocal_Eat_Heavy_01}}
 {id=WavePlayer_12, properties {soundWaveAssetPtr=&/Burt_Vocal_Eat_Heavy_05}}
 {id=WavePlayer_13, properties {soundWaveAssetPtr=&/Burt_Vocal_Eat_Heavy_06}}
 {id=WavePlayer_14, properties {soundWaveAssetPtr=&/Burt_Vocal_Eat_Heavy_07}}
 {id=WavePlayer_18, properties {soundWaveAssetPtr=&/Burt_Vocal_Eat_Heavy_08}}
 {id=WavePlayer_2, properties {soundWaveAssetPtr=&/Burt_Vocal_Eat_Heavy_02}}
 {id=WavePlayer_20, properties {soundWaveAssetPtr=&/Burt_Vocal_Eat_Light_01}}
 {id=WavePlayer_21, properties {soundWaveAssetPtr=&/Burt_Vocal_Eat_Light_02}}
 {id=WavePlayer_22, properties {soundWaveAssetPtr=&/Burt_Vocal_Eat_Light_04}}
 {id=WavePlayer_23, properties {soundWaveAssetPtr=&/Burt_Vocal_Eat_Light_03}}
 {id=WavePlayer_24, properties {soundWaveAssetPtr=&/Burt_Vocal_Eat_Light_05}}
 {id=WavePlayer_25, properties {soundWaveAssetPtr=&/Burt_Vocal_Eat_Light_06}}
 {id=WavePlayer_26, properties {soundWaveAssetPtr=&/Burt_Vocal_Eat_Light_07}}
 {id=WavePlayer_27, properties {soundWaveAssetPtr=&/Burt_Vocal_Eat_Light_08}}
 {id=WavePlayer_3, properties {soundWaveAssetPtr=&/Burt_Vocal_Eat_Heavy_03}}
 {id=WavePlayer_30, properties {soundWaveAssetPtr=&Burt_Vocal_Eat_Light_07}}
 {id=WavePlayer_31, properties {soundWaveAssetPtr=&Burt_Vocal_Eat_Light_08}}
 {id=WavePlayer_4, properties {soundWaveAssetPtr=&/Burt_Vocal_Eat_Heavy_04}}
 {id=WavePlayer_5, properties {soundWaveAssetPtr=&Burt_Vocal_Eat_Heavy_02}}
]
properties {
 attenuationSettings=/Irwin/Sounds/Settings/Prey_Vocalization_Att
 maxDistance=6350
 pitchMultiplier=1.1
 soundClassObject=/Game/Sounds/FortSoundClassesAndModes/Fort_Creatures_Vocalization_Normal
 volumeMultiplier=1.9
 concurrencySet [&Athena_Footstep_Local_Conc, &Athena_Footstep_Global_Conc]
}
