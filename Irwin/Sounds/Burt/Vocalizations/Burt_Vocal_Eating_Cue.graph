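# Sound cue for Burt's eating vocalization, a creature sound using prey vocalization attenuation.
# A Mixer layers two equally weighted random picks, one from four eating waves and one from six short eat waves.
# A root Modulator randomizes pitch and volume on the mixed result, on top of raised cue-level multipliers.
# It routes to the creature vocalization sound class and uses Athena footstep local and global concurrency sets.
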SoundCue /Irwin/Sounds/Burt/Vocalizations/Burt_Vocal_Eating_Cue
root=Modulator_0
nodes [
 {id=Mixer_0, children [Random_0,Random_3], properties {inputVolume [1,1]}}
 {id=Modulator_0, children [Mixer_0], properties {pitchMax=1.2, pitchMin=1, volumeMax=1.1, volumeMin=0.9}}
 {id=Random_0, children [WavePlayer_6,WavePlayer_7,WavePlayer_8,WavePlayer_9], properties {weights [1,1,1,1]}}
 {
  id=Random_3
  children [WavePlayer_10, WavePlayer_11, WavePlayer_15, WavePlayer_16, WavePlayer_17, WavePlayer_19]
  properties {weights [1,1,1,1,1,1]}
 }
 {id=WavePlayer_10, properties {soundWaveAssetPtr=&/Burt_Vocal_Eat_Short_01}}
 {id=WavePlayer_11, properties {soundWaveAssetPtr=&/Burt_Vocal_Eat_Short_03}}
 {id=WavePlayer_15, properties {soundWaveAssetPtr=&/Burt_Vocal_Eat_Short_02}}
 {id=WavePlayer_16, properties {soundWaveAssetPtr=&/Burt_Vocal_Eat_Short_04}}
 {id=WavePlayer_17, properties {soundWaveAssetPtr=&/Burt_Vocal_Eat_Short_05}}
 {id=WavePlayer_19, properties {soundWaveAssetPtr=&/Burt_Vocal_Eat_Short_06}}
 {id=WavePlayer_6, properties {soundWaveAssetPtr=&/Burt_Vocal_Eating_01}}
 {id=WavePlayer_7, properties {soundWaveAssetPtr=&/Burt_Vocal_Eating_02}}
 {id=WavePlayer_8, properties {soundWaveAssetPtr=&/Burt_Vocal_Eating_03}}
 {id=WavePlayer_9, properties {soundWaveAssetPtr=&/Burt_Vocal_Eating_04}}
]
properties {
 attenuationSettings=/Irwin/Sounds/Settings/Prey_Vocalization_Att
 maxDistance=6350
 pitchMultiplier=1.1
 soundClassObject=/Game/Sounds/FortSoundClassesAndModes/Fort_Creatures_Vocalization_Normal
 volumeMultiplier=1.9
 concurrencySet [&Athena_Footstep_Local_Conc, &Athena_Footstep_Global_Conc]
}
