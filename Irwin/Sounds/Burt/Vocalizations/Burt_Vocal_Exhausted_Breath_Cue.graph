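# Sound cue for the exhausted breathing vocalization of Burt, an Irwin creature, played at half volume.
# A Random node picks one of seven equally weighted breath recordings, and a Modulator lowers its pitch and slightly varies pitch and volume.
# It uses close-range prey vocalization attenuation and a tactical creature vocalization sound class.
# Its concurrency is governed by the local and global Athena footstep concurrency sets rather than vocalization-specific ones.
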SoundCue /Irwin/Sounds/Burt/Vocalizations/Burt_Vocal_Exhausted_Breath_Cue
root=Modulator_0
nodes [
 {id=Modulator_0, children [Random_0], properties {pitchMax=0.85, pitchMin=0.8, volumeMax=1.1, volumeMin=0.9}}
 {
  id=Random_0
  children [WavePlayer_0,WavePlayer_1,WavePlayer_2,WavePlayer_3,WavePlayer_4,WavePlayer_5,WavePlayer_6]
  properties {weights [1,1,1,1,1,1,1]}
 }
 {id=WavePlayer_0, properties {soundWaveAssetPtr=&/Burt_Vocal_Exhausted_Breath_01}}
 {id=WavePlayer_1, properties {soundWaveAssetPtr=&/Burt_Vocal_Exhausted_Breath_02}}
 {id=WavePlayer_2, properties {soundWaveAssetPtr=&/Burt_Vocal_Exhausted_Breath_03}}
 {id=WavePlayer_3, properties {soundWaveAssetPtr=&/Burt_Vocal_Exhausted_Breath_04}}
 {id=WavePlayer_4, properties {soundWaveAssetPtr=&/Burt_Vocal_Exhausted_Breath_05}}
 {id=WavePlayer_5, properties {soundWaveAssetPtr=&/Burt_Vocal_Exhausted_Breath_06}}
 {id=WavePlayer_6, properties {soundWaveAssetPtr=&/Burt_Vocal_Exhausted_Breath_07}}
]
properties {
 attenuationSettings=/Irwin/Sounds/Settings/Prey_Vocalization_Close_Att
 maxDistance=2150
 soundClassObject=&Fort_Creatures_Vocalization_Tactical
 volumeMultiplier=0.5
 concurrencySet [&Athena_Footstep_Local_Conc, &Athena_Footstep_Global_Conc]
}
