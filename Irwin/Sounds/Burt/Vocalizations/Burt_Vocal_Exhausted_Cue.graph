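# Sound cue for Burt's exhausted vocalization, built from reused chase breathing and idle ambient recordings.
# A mixer layers a random chase breathing clip with a random idle ambient clip that starts after a short fixed delay.
# A root modulator lowers the pitch and slightly varies volume.
# It uses prey vocalization attenuation, the tactical creature vocalization sound class, and borrows footstep concurrency sets.
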SoundCue /Irwin/Sounds/Burt/Vocalizations/Burt_Vocal_Exhausted_Cue
root=Modulator_0
nodes [
 {id=Delay_0, children [Random_2], properties {delayMax=0.5, delayMin=0.5}}
 {id=Mixer_0, children [Random_3,Delay_0], properties {inputVolume [1,1]}}
 {id=Modulator_0, children [Mixer_0], properties {pitchMax=0.85, pitchMin=0.8, volumeMax=1.1, volumeMin=0.9}}
 {id=Random_2, children [WavePlayer_10, WavePlayer_11, WavePlayer_12, WavePlayer_13], properties {weights [1,1,1,1]}}
 {id=Random_3, children [WavePlayer_27, WavePlayer_28, WavePlayer_29, WavePlayer_30], properties {weights [1,1,1,1]}}
 {id=WavePlayer_10, properties {soundWaveAssetPtr=&/Burt_Vocal_IdleAmbient_05}}
 {id=WavePlayer_11, properties {soundWaveAssetPtr=&/Burt_Vocal_IdleAmbient_01}}
 {id=WavePlayer_12, properties {soundWaveAssetPtr=&/Burt_Vocal_IdleAmbient_02}}
 {id=WavePlayer_13, properties {soundWaveAssetPtr=&/Burt_Vocal_IdleAmbient_03}}
 {id=WavePlayer_27, properties {soundWaveAssetPtr=&/Burt_Vocal_Chase_Breathing_01}}
 {id=WavePlayer_28, properties {soundWaveAssetPtr=&/Burt_Vocal_Chase_Breathing_03}}
 {id=WavePlayer_29, properties {soundWaveAssetPtr=&/Burt_Vocal_Chase_Breathing_04}}
 {id=WavePlayer_30, properties {soundWaveAssetPtr=&/Burt_Vocal_Chase_Breathing_05}}
]
properties {
 attenuationSettings=/Irwin/Sounds/Settings/Prey_Vocalization_Att
 maxDistance=6350
 soundClassObject=&Fort_Creatures_Vocalization_Tactical
 volumeMultiplier=2
 concurrencySet [&Athena_Footstep_Local_Conc, &Athena_Footstep_Global_Conc]
}
