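# Sound cue for Burt's idle ambient vocalizations (breaths, grunts and ambient calls), using prey vocalization attenuation and footstep concurrency sets.
# A top-level weighted random plays a vocal about one time in three and otherwise stays silent, under slight pitch and volume modulation.
# A quality-level node routes the first two slots to a full pool randomized without replacement and the last two to a five-clip subset.
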SoundCue /Irwin/Sounds/Burt/Vocalizations/Burt_Vocal_IdleAmbient_Cue
root=Modulator_0
nodes [
 {id=Modulator_0, children [Random_0], properties {pitchMax=1.1, volumeMax=1.1, volumeMin=0.9}}
 {id=QualityLevel_0, children [Random_3,Random_3,Random_5,Random_5]}
 {
  id=Random_0
  children [QualityLevel_0, null]
  properties {bRandomizeWithoutReplacement=false, bShouldExcludeFromBranchCulling=true, weights [0.5,1]}
 }
 {
  id=Random_3
  children [
   WavePlayer_0
   WavePlayer_5
   WavePlayer_7
   WavePlayer_8
   WavePlayer_1
   WavePlayer_3
   WavePlayer_4
   WavePlayer_10
   WavePlayer_11
   WavePlayer_12
   WavePlayer_13
   WavePlayer_14
   WavePlayer_15
   WavePlayer_16
   WavePlayer_17
   WavePlayer_18
  ]
  properties {weights [1,1,1,1,1,1,1,1,1,1,1,1,1,1,1,1]}
 }
 {
  id=Random_5
  children [WavePlayer_33, WavePlayer_37, WavePlayer_38, WavePlayer_43, WavePlayer_44]
  properties {bRandomizeWithoutReplacement=false, bShouldExcludeFromBranchCulling=true, weights [1,1,1,1,1]}
 }
 {id=WavePlayer_0, properties {soundWaveAssetPtr=&/Burt_Vocal_IdleAmbient_01}}
 {id=WavePlayer_1, properties {soundWaveAssetPtr=&/Burt_Vocal_IdleBreathTriple_01}}
 {id=WavePlayer_10, properties {soundWaveAssetPtr=&/Burt_Vocal_IdleGrunt_03}}
 {id=WavePlayer_11, properties {soundWaveAssetPtr=&/Burt_Vocal_Idle_Breathing_04}}
 {id=WavePlayer_12, properties {soundWaveAssetPtr=&/Burt_Vocal_Idle_Breathing_01}}
 {id=WavePlayer_13, properties {soundWaveAssetPtr=&/Burt_Vocal_Idle_Breathing_02}}
 {id=WavePlayer_14, properties {soundWaveAssetPtr=&/Burt_Vocal_Idle_Breathing_03}}
 {id=WavePlayer_15, properties {soundWaveAssetPtr=&/Burt_Vocal_Idle_Grunting_01}}
 {id=WavePlayer_16, properties {soundWaveAssetPtr=&/Burt_Vocal_Idle_Grunting_02}}
 {id=WavePlayer_17, properties {soundWaveAssetPtr=&/Burt_Vocal_Idle_Grunting_04}}
 {id=WavePlayer_18, properties {soundWaveAssetPtr=&/Burt_Vocal_Idle_Grunting_06}}
 {id=WavePlayer_3, properties {soundWaveAssetPtr=&/Burt_Vocal_IdleBreathTriple_03}}
 {id=WavePlayer_33, properties {soundWaveAssetPtr=&/Burt_Vocal_IdleAmbient_02}}
 {id=WavePlayer_37, properties {soundWaveAssetPtr=&Burt_Vocal_IdleBreathTriple_03}}
 {id=WavePlayer_38, properties {soundWaveAssetPtr=&/Burt_Vocal_IdleGrunt_01}}
 {id=WavePlayer_4, properties {soundWaveAssetPtr=&Burt_Vocal_IdleGrunt_01}}
 {id=WavePlayer_43, properties {soundWaveAssetPtr=&Burt_Vocal_Idle_Breathing_03}}
 {id=WavePlayer_44, properties {soundWaveAssetPtr=&Burt_Vocal_Idle_Grunting_01}}
 {id=WavePlayer_5, properties {soundWaveAssetPtr=&Burt_Vocal_IdleAmbient_02}}
 {id=WavePlayer_7, properties {soundWaveAssetPtr=&/Burt_Vocal_IdleAmbient_04}}
 {id=WavePlayer_8, properties {soundWaveAssetPtr=&/Burt_Vocal_IdleAmbient_05}}
]
properties {
 attenuationSettings=/Irwin/Sounds/Settings/Prey_Vocalization_Att
 maxDistance=6350
 soundClassObject=&Fort_Creatures_Vocalization_Tactical
 volumeMultiplier=1.4
 concurrencySet [&Athena_Footstep_Local_Conc, &Athena_Footstep_Global_Conc]
}
