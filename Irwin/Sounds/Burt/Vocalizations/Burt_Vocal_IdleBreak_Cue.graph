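# Sound cue for Burt's idle-break vocalization, playing a random jump, phrase or snort clip with modulated pitch and volume.
# A quality-level node routes the first two levels to a nine-clip random pool and the last two to a five-clip subset excluded from branch culling.
# It uses prey vocalization attenuation and the creature vocalization sound class, but Athena footstep local and global concurrency.
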SoundCue /Irwin/Sounds/Burt/Vocalizations/Burt_Vocal_IdleBreak_Cue
root=Modulator_0
nodes [
 {id=Modulator_0, children [QualityLevel_0], properties {pitchMax=1.1, volumeMax=1.1, volumeMin=0.9}}
 {id=QualityLevel_0, children [Random_0,Random_0,Random_1,Random_1]}
 {
  id=Random_0
  children [WavePlayer_0, WavePlayer_2, WavePlayer_3, WavePlayer_6, WavePlayer_8, WavePlayer_9, WavePlayer_10, WavePlayer_11, WavePlayer_12]
  properties {weights [1,1,1,1,1,1,1,1,1]}
 }
 {
  id=Random_1
  children [WavePlayer_5, WavePlayer_1, WavePlayer_13, WavePlayer_16, WavePlayer_18]
  properties {bShouldExcludeFromBranchCulling=true, weights [1,1,1,1,1]}
 }
 {id=WavePlayer_0, properties {soundWaveAssetPtr=&/Burt_Vocal_Jump_01}}
 {id=WavePlayer_1, properties {soundWaveAssetPtr=&/Burt_Vocal_Phrase_03}}
 {id=WavePlayer_10, properties {soundWaveAssetPtr=&/Burt_Vocal_Idle_Snorting_01}}
 {id=WavePlayer_11, properties {soundWaveAssetPtr=&/Burt_Vocal_Jump_03}}
 {id=WavePlayer_12, properties {soundWaveAssetPtr=&/Burt_Vocal_SnortSingle_01}}
 {id=WavePlayer_13, properties {soundWaveAssetPtr=&/Burt_Vocal_SnortSingle_05}}
 {id=WavePlayer_16, properties {soundWaveAssetPtr=&Burt_Vocal_Idle_Snorting_01}}
 {id=WavePlayer_18, properties {soundWaveAssetPtr=&Burt_Vocal_SnortSingle_01}}
 {id=WavePlayer_2, properties {soundWaveAssetPtr=&Burt_Vocal_Phrase_03}}
 {id=WavePlayer_3, properties {soundWaveAssetPtr=&/Burt_Vocal_Phrase_01}}
 {id=WavePlayer_5, properties {soundWaveAssetPtr=&Burt_Vocal_Jump_01}}
 {id=WavePlayer_6, properties {soundWaveAssetPtr=&Burt_Vocal_SnortSingle_05}}
 {id=WavePlayer_8, properties {soundWaveAssetPtr=&/Burt_Vocal_Idle_Snorting_08}}
 {id=WavePlayer_9, properties {soundWaveAssetPtr=&/Burt_Vocal_Idle_Snorting_06}}
]
properties {
 attenuationSettings=/Irwin/Sounds/Settings/Prey_Vocalization_Att
 maxDistance=6350
 soundClassObject=/Game/Sounds/FortSoundClassesAndModes/Fort_Creatures_Vocalization_Normal
 volumeMultiplier=1.7
 concurrencySet [&Athena_Footstep_Local_Conc, &Athena_Footstep_Global_Conc]
}
