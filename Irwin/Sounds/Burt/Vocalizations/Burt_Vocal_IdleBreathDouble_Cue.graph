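# Sound cue for Burt's idle breath vocalization, playing randomized snorts under concurrency limits that include a per-owner VO set.
# An IsLocal branch plays full level when true, otherwise an IsRiddenByEnemy branch selects enemy-rider class and attenuation or quieter creature settings, with full level if unset.
# All paths share a pitch and volume modulator feeding a quality switch, which cuts the random pool from eight waves to four on two quality levels.
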SoundCue /Irwin/Sounds/Burt/Vocalizations/Burt_Vocal_IdleBreathDouble_Cue
root=Branch_2
nodes [
 {id=Attenuation_0, children [Modulator_1], properties {attenuationSettings=/Irwin/Sounds/Settings/Prey_Vocalization_Att}}
 {id=Attenuation_2, children [Modulator_1], properties {attenuationSettings=/Game/Sounds/Attenuation/Fort_Riding_Enemy_VO_Att}}
 {id=Attenuation_3, children [Modulator_1], properties {attenuationSettings=&Prey_Vocalization_Att}}
 {id=Attenuation_4, children [Modulator_1], properties {attenuationSettings=&Prey_Vocalization_Att}}
 {id=Branch_1, children [Modulator_3,Modulator_2,SoundClass_0], properties {boolParameterName=IsRiddenByEnemy}}
 {id=Branch_2, children [SoundClass_4,Branch_1,Branch_1], properties {boolParameterName=IsLocal}}
 {id=Modulator_1, children [QualityLevel_0], properties {pitchMax=1.1, pitchMin=0.9, volumeMax=1, volumeMin=0.5}}
 {id=Modulator_2, children [SoundClass_3], properties {pitchMax=1, pitchMin=1, volumeMax=0.3, volumeMin=0.3}}
 {id=Modulator_3, children [SoundClass_2], properties {pitchMax=1, pitchMin=1, volumeMax=0.7, volumeMin=0.7}}
 {id=QualityLevel_0, children [Random_1,Random_1,Random_0,Random_0]}
 {
  id=Random_0
  children [WavePlayer_0,WavePlayer_1,WavePlayer_2,WavePlayer_3]
  properties {bShouldExcludeFromBranchCulling=true, weights [1,1,1,1]}
 }
 {
  id=Random_1
  children [WavePlayer_7, WavePlayer_8, WavePlayer_12, WavePlayer_13, WavePlayer_14, WavePlayer_15, WavePlayer_16, WavePlayer_17]
  properties {weights [1,1,1,1,1,1,1,1]}
 }
 {
  id=SoundClass_0
  children [Attenuation_0]
  properties {soundClassOverride=&Fort_Creatures_Vocalization_Tactical}
 }
 {
  id=SoundClass_2
  children [Attenuation_2]
  properties {soundClassOverride=&Fort_Player_Foley_3P_Enemy_Athena}
 }
 {
  id=SoundClass_3
  children [Attenuation_3]
  properties {soundClassOverride=&Fort_Creatures_Vocalization_Tactical}
 }
 {
  id=SoundClass_4
  children [Attenuation_4]
  properties {soundClassOverride=&Fort_Creatures_Vocalization_Tactical}
 }
 {id=WavePlayer_0, properties {soundWaveAssetPtr=&/Burt_Vocal_Idle_Snorting_01}}
 {id=WavePlayer_1, properties {soundWaveAssetPtr=&/Burt_Vocal_Idle_Snorting_02}}
 {id=WavePlayer_12, properties {soundWaveAssetPtr=&/Burt_Vocal_Idle_Snorting_03}}
 {id=WavePlayer_13, properties {soundWaveAssetPtr=&/Burt_Vocal_Idle_Snorting_04}}
 {id=WavePlayer_14, properties {soundWaveAssetPtr=&/Burt_Vocal_Idle_Snorting_05}}
 {id=WavePlayer_15, properties {soundWaveAssetPtr=&/Burt_Vocal_Idle_Snorting_06}}
 {id=WavePlayer_16, properties {soundWaveAssetPtr=&/Burt_Vocal_Idle_Snorting_07}}
 {id=WavePlayer_17, properties {soundWaveAssetPtr=&/Burt_Vocal_Idle_Snorting_08}}
 {id=WavePlayer_2, properties {soundWaveAssetPtr=&Burt_Vocal_Idle_Snorting_03}}
 {id=WavePlayer_3, properties {soundWaveAssetPtr=&Burt_Vocal_Idle_Snorting_04}}
 {id=WavePlayer_7, properties {soundWaveAssetPtr=&Burt_Vocal_Idle_Snorting_01}}
 {id=WavePlayer_8, properties {soundWaveAssetPtr=&Burt_Vocal_Idle_Snorting_02}}
]
properties {
 maxDistance=9000
 volumeMultiplier=1.4
 concurrencySet [
  &Athena_Footstep_Local_Conc
  &Athena_Footstep_Global_Conc
  /Irwin/Sounds/Settings/Conc/SetOneLimitToOwner_VO_Conc
 ]
}
