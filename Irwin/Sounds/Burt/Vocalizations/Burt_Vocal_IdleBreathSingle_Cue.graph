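# Sound cue for a single idle breathing vocalization from Burt, an Irwin creature that uses prey vocalization attenuation.
# A quality switch feeds either a six-wave random pool (first two levels) or a three-wave pool (last two), the latter excluded from branch culling.
# A modulator adds slight pitch and volume variation.
# The cue routes to the tactical creature vocalization class but uses local and global footstep concurrency.
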SoundCue /Irwin/Sounds/Burt/Vocalizations/Burt_Vocal_IdleBreathSingle_Cue
root=Modulator_0
nodes [
 {id=Modulator_0, children [QualityLevel_0], properties {pitchMax=1.1, pitchMin=0.9, volumeMax=1.1, volumeMin=0.9}}
 {id=QualityLevel_0, children [Random_4,Random_4,Random_0,Random_0]}
 {id=Random_0, children [WavePlayer_0,WavePlayer_1,WavePlayer_2], properties {bShouldExcludeFromBranchCulling=true, weights [1,1,1]}}
 {
  id=Random_4
  children [WavePlayer_24, WavePlayer_25, WavePlayer_26, WavePlayer_27, WavePlayer_28, WavePlayer_29]
  properties {weights [1,1,1,1,1,1]}
 }
 {id=WavePlayer_0, properties {soundWaveAssetPtr=&/Burt_Vocal_Idle_Breathing_01}}
 {id=WavePlayer_1, properties {soundWaveAssetPtr=&/Burt_Vocal_Idle_Breathing_02}}
 {id=WavePlayer_2, properties {soundWaveAssetPtr=&/Burt_Vocal_Idle_Breathing_03}}
 {id=WavePlayer_24, properties {soundWaveAssetPtr=&Burt_Vocal_Idle_Breathing_01}}
 {id=WavePlayer_25, properties {soundWaveAssetPtr=&Burt_Vocal_Idle_Breathing_02}}
 {id=WavePlayer_26, properties {soundWaveAssetPtr=&Burt_Vocal_Idle_Breathing_03}}
 {id=WavePlayer_27, properties {soundWaveAssetPtr=&/Burt_Vocal_Idle_Breathing_04}}
 {id=WavePlayer_28, properties {soundWaveAssetPtr=&/Burt_Vocal_Idle_Breathing_05}}
 {id=WavePlayer_29, properties {soundWaveAssetPtr=&/Burt_Vocal_Idle_Breathing_06}}
]
properties {
 attenuationSettings=/Irwin/Sounds/Settings/Prey_Vocalization_Att
 maxDistance=6350
 soundClassObject=&Fort_Creatures_Vocalization_Tactical
 volumeMultiplier=1
 concurrencySet [&Athena_Footstep_Local_Conc, &Athena_Footstep_Global_Conc]
}
