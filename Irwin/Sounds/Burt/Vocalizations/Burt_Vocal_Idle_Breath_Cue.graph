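# Sound cue for the idle breathing vocalization of Burt, a rideable creature, which grows more strained as fatigue rises.
# IsLocal and IsRiddenByEnemy branches pick sound class, attenuation and volume, boosting enemy-ridden mounts, then share pitch and volume randomization.
# A quality switch selects between two FatigueLevel crossfades from idle through tiring to exhausted or agitated sounds.
# Random nodes include empty slots for occasional silence, and concurrency sets limit overlap.
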SoundCue /Irwin/Sounds/Burt/Vocalizations/Burt_Vocal_Idle_Breath_Cue
root=Branch_2
nodes [
 {id=Attenuation_0, children [Modulator_1], properties {attenuationSettings=/Irwin/Sounds/Settings/Prey_Vocalization_Att}}
 {id=Attenuation_2, children [Modulator_1], properties {attenuationSettings=/Game/Sounds/Attenuation/Fort_Riding_Enemy_VO_Att}}
 {id=Attenuation_3, children [Modulator_1], properties {attenuationSettings=&Prey_Vocalization_Att}}
 {id=Attenuation_4, children [Modulator_1], properties {attenuationSettings=&Prey_Vocalization_Att}}
 {id=Branch_1, children [Modulator_3,Modulator_2,SoundClass_0], properties {boolParameterName=IsRiddenByEnemy}}
 {id=Branch_2, children [SoundClass_4,Branch_1,Branch_1], properties {boolParameterName=IsLocal}}
 {id=Modulator_1, children [QualityLevel_1], properties {pitchMax=1.1, pitchMin=0.9, volumeMax=1, volumeMin=0.5}}
 {id=Modulator_2, children [SoundClass_3], properties {pitchMax=1, pitchMin=1, volumeMax=0.3, volumeMin=0.3}}
 {id=Modulator_3, children [SoundClass_2], properties {pitchMax=1, pitchMin=1, volumeMax=1.2, volumeMin=1.2}}
 {id=Modulator_4, children [Random_0], properties {pitchMax=1, pitchMin=1, volumeMax=1.2, volumeMin=1}}
 {id=Modulator_5, children [Random_14], properties {pitchMax=1, pitchMin=1, volumeMax=2.2, volumeMin=2}}
 {
  id=ParamCrossFade_1
  children [Random_7,Random_9,Random_13]
  properties {
   paramName=FatigueLevel
   crossFadeInput [
    {fadeInDistanceEnd=0, fadeInDistanceStart=0, fadeOutDistanceEnd=0.85, fadeOutDistanceStart=0.82, volume=2.4}
    {fadeInDistanceEnd=0.85, fadeInDistanceStart=0.82, fadeOutDistanceEnd=0.96, fadeOutDistanceStart=0.95, volume=1.2}
    {fadeInDistanceEnd=0.96, fadeInDistanceStart=0.95, fadeOutDistanceEnd=1, fadeOutDistanceStart=1, volume=1.2}
   ]
  }
 }
 {
  id=ParamCrossFade_3
  children [Random_4,Random_5,Random_16]
  properties {
   paramName=FatigueLevel
   crossFadeInput [
    {fadeInDistanceEnd=0, fadeInDistanceStart=0, fadeOutDistanceEnd=0.85, fadeOutDistanceStart=0.82, volume=1}
    {fadeInDistanceEnd=0.85, fadeInDistanceStart=0.82, fadeOutDistanceEnd=0.96, fadeOutDistanceStart=0.95, volume=2}
    {fadeInDistanceEnd=0.96, fadeInDistanceStart=0.95, fadeOutDistanceEnd=1, fadeOutDistanceStart=1, volume=2}
   ]
  }
 }
 {id=QualityLevel_1, children [ParamCrossFade_1, ParamCrossFade_1, ParamCrossFade_3, ParamCrossFade_3]}
 {id=Random_0, children [WavePlayer_45, WavePlayer_46, WavePlayer_47, WavePlayer_48, WavePlayer_49], properties {weights [1,1,1,1,1]}}
 {id=Random_12, children [WavePlayer_10, WavePlayer_11, WavePlayer_12, WavePlayer_13], properties {weights [1,1,1,1]}}
 {id=Random_13, children [Modulator_4,Random_12,null], properties {bRandomizeWithoutReplacement=false, weights [1,0.8,0.7]}}
 {id=Random_14, children [WavePlayer_54, WavePlayer_55, WavePlayer_56, WavePlayer_57], properties {weights [1,1,1,1]}}
 {id=Random_15, children [WavePlayer_50, WavePlayer_51, WavePlayer_52, WavePlayer_53], properties {weights [1,1,1,1]}}
 {id=Random_16, children [Modulator_5,Random_15,null], properties {bRandomizeWithoutReplacement=false, weights [1,0.8,0.7]}}
 {
  id=Random_4
  children [WavePlayer_6,WavePlayer_7,WavePlayer_8,WavePlayer_9,null,null,null]
  properties {bShouldExcludeFromBranchCulling=true, weights [1,1,1,1,1,1,1]}
 }
 {
  id=Random_5
  children [WavePlayer_32, WavePlayer_36, WavePlayer_37, WavePlayer_38]
  properties {bShouldExcludeFromBranchCulling=true, weights [1,1,1,1]}
 }
 {
  id=Random_7
  children [WavePlayer_0,WavePlayer_1,WavePlayer_2,WavePlayer_3,WavePlayer_4,WavePlayer_5,null,null,null,null]
  properties {weights [1,1,1,1,1,1,1,1,1,1]}
 }
 {
  id=Random_9
  children [
   WavePlayer_29
   WavePlayer_30
   WavePlayer_31
   WavePlayer_33
   WavePlayer_34
   WavePlayer_35
   null
   null
   null
   null
   null
   null
   null
   null
   null
   null
  ]
  properties {weights [1,1,1,1,1,1,1,1,1,1,1,1,1,1,1,1]}
 }
 {
  id=SoundClass_0
  children [Attenuation_0]
  properties {soundClassOverride=&Fort_Creatures_Vocalization_Tactical}
 }
 {
  id=SoundClass_2
  children [Attenuation_2]
  properties {soundClassOverride=&Fort_Player_Foley_3P_Enemy_Athena}
 }
 {
  id=SoundClass_3
  children [Attenuation_3]
  properties {soundClassOverride=&Fort_Creatures_Vocalization_Tactical}
 }
 {
  id=SoundClass_4
  children [Attenuation_4]
  properties {soundClassOverride=&Fort_Creatures_Vocalization_Tactical}
 }
 {id=WavePlayer_0, properties {soundWaveAssetPtr=&/Burt_Vocal_Idle_Breathing_01}}
 {id=WavePlayer_1, properties {soundWaveAssetPtr=&/Burt_Vocal_Idle_Breathing_02}}
 {id=WavePlayer_10, properties {soundWaveAssetPtr=&/Burt_Vocal_Agitated_01}}
 {id=WavePlayer_11, properties {soundWaveAssetPtr=&/Burt_Vocal_Agitated_02}}
 {id=WavePlayer_12, properties {soundWaveAssetPtr=&/Burt_Vocal_Agitated_06}}
 {id=WavePlayer_13, properties {soundWaveAssetPtr=&/Burt_Vocal_Agitated_07}}
 {id=WavePlayer_2, properties {soundWaveAssetPtr=&/Burt_Vocal_Idle_Breathing_03}}
 {id=WavePlayer_29, properties {soundWaveAssetPtr=&/Burt_Vocal_Sprint_Tiring_Breathing_01}}
 {id=WavePlayer_3, properties {soundWaveAssetPtr=&/Burt_Vocal_Idle_Breathing_04}}
 {id=WavePlayer_30, properties {soundWaveAssetPtr=&/Burt_Vocal_Sprint_Tiring_Breathing_02}}
 {id=WavePlayer_31, properties {soundWaveAssetPtr=&/Burt_Vocal_Sprint_Tiring_Breathing_03}}
 {id=WavePlayer_32, properties {soundWaveAssetPtr=&Burt_Vocal_Sprint_Tiring_Breathing_01}}
 {id=WavePlayer_33, properties {soundWaveAssetPtr=&/Burt_Vocal_Sprint_Tiring_Breathing_05}}
 {id=WavePlayer_34, properties {soundWaveAssetPtr=&/Burt_Vocal_Sprint_Tiring_Breathing_06}}
 {id=WavePlayer_35, properties {soundWaveAssetPtr=&/Burt_Vocal_Sprint_Tiring_Breathing_07}}
 {id=WavePlayer_36, properties {soundWaveAssetPtr=&Burt_Vocal_Sprint_Tiring_Breathing_02}}
 {id=WavePlayer_37, properties {soundWaveAssetPtr=&Burt_Vocal_Sprint_Tiring_Breathing_03}}
 {id=WavePlayer_38, properties {soundWaveAssetPtr=&Burt_Vocal_Sprint_Tiring_Breathing_05}}
 {id=WavePlayer_4, properties {soundWaveAssetPtr=&/Burt_Vocal_Idle_Breathing_05}}
 {id=WavePlayer_45, properties {soundWaveAssetPtr=&Burt_Vocal_Sprint_Tiring_Breathing_01}}
 {id=WavePlayer_46, properties {soundWaveAssetPtr=&Burt_Vocal_Sprint_Tiring_Breathing_02}}
 {id=WavePlayer_47, properties {soundWaveAssetPtr=&Burt_Vocal_Sprint_Tiring_Breathing_03}}
 {id=WavePlayer_48, properties {soundWaveAssetPtr=&Burt_Vocal_Sprint_Tiring_Breathing_05}}
 {id=WavePlayer_49, properties {soundWaveAssetPtr=&Burt_Vocal_Sprint_Tiring_Breathing_06}}
 {id=WavePlayer_5, properties {soundWaveAssetPtr=&/Burt_Vocal_Idle_Breathing_06}}
 {id=WavePlayer_50, properties {soundWaveAssetPtr=&Burt_Vocal_Agitated_01}}
 {id=WavePlayer_51, properties {soundWaveAssetPtr=&Burt_Vocal_Agitated_02}}
 {id=WavePlayer_52, properties {soundWaveAssetPtr=&Burt_Vocal_Agitated_06}}
 {id=WavePlayer_53, properties {soundWaveAssetPtr=&Burt_Vocal_Agitated_07}}
 {id=WavePlayer_54, properties {soundWaveAssetPtr=&Burt_Vocal_Sprint_Tiring_Breathing_01}}
 {id=WavePlayer_55, properties {soundWaveAssetPtr=&Burt_Vocal_Sprint_Tiring_Breathing_02}}
 {id=WavePlayer_56, properties {soundWaveAssetPtr=&Burt_Vocal_Sprint_Tiring_Breathing_03}}
 {id=WavePlayer_57, properties {soundWaveAssetPtr=&Burt_Vocal_Sprint_Tiring_Breathing_05}}
 {id=WavePlayer_6, properties {soundWaveAssetPtr=&Burt_Vocal_Idle_Breathing_01}}
 {id=WavePlayer_7, properties {soundWaveAssetPtr=&Burt_Vocal_Idle_Breathing_02}}
 {id=WavePlayer_8, properties {soundWaveAssetPtr=&Burt_Vocal_Idle_Breathing_03}}
 {id=WavePlayer_9, properties {soundWaveAssetPtr=&Burt_Vocal_Idle_Breathing_04}}
]
properties {
 maxDistance=9000
 volumeMultiplier=1.6
 concurrencySet [
  &Athena_Footstep_Local_Conc
  &Athena_Footstep_Global_Conc
  /Game/Sounds/Concurrency_Rules/LimitToOnePreventNewQuick
 ]
}
