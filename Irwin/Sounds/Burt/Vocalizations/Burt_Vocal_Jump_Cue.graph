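# Sound cue for Burt's jump vocalization, a creature that can be ridden.
# An IsLocal branch gives local playback its own level, otherwise deferring to IsRiddenByEnemy, which selects enemy-rider foley class and attenuation, much quieter creature vocalization, or unscaled vocalization when unset.
# All paths share a slight pitch/volume modulator over an equal-weight random of four waves plus an empty slot, so some jumps are silent.
# Footstep and limit-to-one concurrency sets cap overlapping plays.
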SoundCue /Irwin/Sounds/Burt/Vocalizations/Burt_Vocal_Jump_Cue
root=Branch_2
nodes [
 {id=Attenuation_0, children [Modulator_0], properties {attenuationSettings=/Irwin/Sounds/Settings/Prey_Vocalization_Att}}
 {id=Attenuation_2, children [Modulator_0], properties {attenuationSettings=/Game/Sounds/Attenuation/Fort_Riding_Enemy_Att}}
 {id=Attenuation_3, children [Modulator_0], properties {attenuationSettings=&Prey_Vocalization_Att}}
 {id=Attenuation_4, children [Modulator_0], properties {attenuationSettings=&Prey_Vocalization_Att}}
 {id=Branch_1, children [Modulator_3,Modulator_2,SoundClass_0], properties {boolParameterName=IsRiddenByEnemy}}
 {id=Branch_2, children [Modulator_1,Branch_1,Branch_1], properties {boolParameterName=IsLocal}}
 {id=Modulator_0, children [Random_1], properties {pitchMax=1.1, pitchMin=0.9, volumeMax=1, volumeMin=0.9}}
 {id=Modulator_1, children [SoundClass_4], properties {pitchMax=1, pitchMin=1, volumeMax=0.85, volumeMin=0.8}}
 {id=Modulator_2, children [SoundClass_3], properties {pitchMax=1, pitchMin=1, volumeMax=0.1, volumeMin=0.1}}
 {id=Modulator_3, children [SoundClass_2], properties {pitchMax=1, pitchMin=1, volumeMax=0.7, volumeMin=0.7}}
 {
  id=Random_1
  children [WavePlayer_0,WavePlayer_1,WavePlayer_2,WavePlayer_3,null]
  properties {bShouldExcludeFromBranchCulling=true, weights [1,1,1,1,1]}
 }
 {
  id=SoundClass_0
  children [Attenuation_0]
  properties {soundClassOverride=&Fort_Creatures_Vocalization_Tactical}
 }
 {
  id=SoundClass_2
  children [Attenuation_2]
  properties {soundClassOverride=&Fort_Player_Foley_3P_Enemy_Athena}
 }
 {
  id=SoundClass_3
  children [Attenuation_3]
  properties {soundClassOverride=&Fort_Creatures_Vocalization_Tactical}
 }
 {
  id=SoundClass_4
  children [Attenuation_4]
  properties {soundClassOverride=&Fort_Creatures_Vocalization_Tactical}
 }
 {id=WavePlayer_0, properties {soundWaveAssetPtr=&/Burt_Vocal_Jump_01}}
 {id=WavePlayer_1, properties {soundWaveAssetPtr=&/Burt_Vocal_Jump_02}}
 {id=WavePlayer_2, properties {soundWaveAssetPtr=&/Burt_Vocal_Jump_03}}
 {id=WavePlayer_3, properties {soundWaveAssetPtr=&/Burt_Vocal_Jump_04}}
]
properties {
 maxDistance=13000
 priority=0.9
 volumeMultiplier=1
 concurrencySet [
  &Athena_Footstep_Global_Conc
  &Athena_Footstep_Local_Conc
  /Game/Sounds/Concurrency_Rules/LimitToOnePreventNew
 ]
}
