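# Burt creature pain vocalization Sound Cue, built from single snort waves with prey vocalization attenuation.
# A short fixed delay feeds pitch and volume modulation, then a quality-level switch into weighted random pools.
# The first two quality slots share a four-snort pool, and the last two share a three-snort pool.
# Concurrency combines Athena local and global footstep limits with a one-per-owner voice limit.
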SoundCue /Irwin/Sounds/Burt/Vocalizations/Burt_Vocal_Pain_Cue
root=Delay_0
nodes [
 {id=Delay_0, children [Modulator_0], properties {delayMax=0.1, delayMin=0.1}}
 {id=Modulator_0, children [QualityLevel_0], properties {pitchMax=1.1, volumeMax=1.2, volumeMin=1}}
 {id=QualityLevel_0, children [Random_1,Random_1,Random_0,Random_0]}
 {id=Random_0, children [WavePlayer_0,WavePlayer_5,WavePlayer_6], properties {weights [1,1,2]}}
 {id=Random_1, children [WavePlayer_1,WavePlayer_2,WavePlayer_3,WavePlayer_4], properties {weights [1,1,2,4]}}
 {id=WavePlayer_0, properties {soundWaveAssetPtr=&/Burt_Vocal_SnortSingle_01}}
 {id=WavePlayer_1, properties {soundWaveAssetPtr=&Burt_Vocal_SnortSingle_01}}
 {id=WavePlayer_2, properties {soundWaveAssetPtr=&/Burt_Vocal_SnortSingle_02}}
 {id=WavePlayer_3, properties {soundWaveAssetPtr=&/Burt_Vocal_SnortSingle_03}}
 {id=WavePlayer_4, properties {soundWaveAssetPtr=&/Burt_Vocal_SnortSingle_04}}
 {id=WavePlayer_5, properties {soundWaveAssetPtr=&Burt_Vocal_SnortSingle_02}}
 {id=WavePlayer_6, properties {soundWaveAssetPtr=&Burt_Vocal_SnortSingle_03}}
]
properties {
 attenuationSettings=/Irwin/Sounds/Settings/Prey_Vocalization_Att
 maxDistance=6350
 priority=0.9
 soundClassObject=&Fort_Creatures_Vocalization_Tactical
 volumeMultiplier=4
 concurrencySet [
  &Athena_Footstep_Local_Conc
  &Athena_Footstep_Global_Conc
  /Irwin/Sounds/Settings/Conc/SetOneLimitToOwner_VO_Conc
 ]
}
