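# Sound cue for Burt's sprint vocalizations while ridden, with breathing that grows more tired and agitated as FatigueLevel rises.
# IsLocal and IsRiddenByEnemy branches select volume, sound class and attenuation, then share a pitch/volume modulator.
# A quality switch picks larger or smaller wave pools, each crossfading three fatigue bands: chase breathing, tiring breathing, then tiring or agitated calls.
# IsEnemy mixes charge-attack calls into the lowest band, and empty random slots add silent gaps.
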SoundCue /Irwin/Sounds/Burt/Vocalizations/Burt_Vocal_PlayerSprint_Cue
root=Branch_2
nodes [
 {id=Attenuation_0, children [Modulator_1], properties {attenuationSettings=/Game/Sounds/Attenuation/Fort_Riding_Enemy_VO_Att}}
 {id=Attenuation_1, children [Modulator_1], properties {attenuationSettings=/Irwin/Sounds/Settings/Prey_Vocalization_Att}}
 {id=Attenuation_3, children [Modulator_1], properties {attenuationSettings=&Prey_Vocalization_Att}}
 {id=Attenuation_4, children [Modulator_1], properties {attenuationSettings=&Prey_Vocalization_Att}}
 {id=Branch_0, children [Modulator_3,Modulator_2,SoundClass_0], properties {boolParameterName=IsRiddenByEnemy}}
 {id=Branch_2, children [Modulator_6,Branch_0,Branch_0], properties {boolParameterName=IsLocal}}
 {id=Branch_3, children [Modulator_0,Random_1,Random_1], properties {boolParameterName=IsEnemy}}
 {id=Branch_4, children [Random_10,Random_7,Random_7], properties {boolParameterName=IsEnemy}}
 {id=Modulator_0, children [Random_2], properties {pitchMax=1, pitchMin=1, volumeMax=0.5, volumeMin=0.5}}
 {id=Modulator_1, children [QualityLevel_0], properties {pitchMax=1.1, pitchMin=0.9, volumeMax=1, volumeMin=0.8}}
 {id=Modulator_2, children [SoundClass_3], properties {pitchMax=1, pitchMin=1, volumeMax=0.3, volumeMin=0.3}}
 {id=Modulator_3, children [SoundClass_2], properties {pitchMax=1, pitchMin=1, volumeMax=0.7, volumeMin=0.7}}
 {id=Modulator_4, children [Random_5], properties {pitchMax=1, pitchMin=1, volumeMax=2.2, volumeMin=2}}
 {id=Modulator_5, children [Random_14], properties {pitchMax=1, pitchMin=1, volumeMax=2.2, volumeMin=2}}
 {id=Modulator_6, children [SoundClass_4], properties {pitchMax=1, pitchMin=1, volumeMax=1, volumeMin=0.6}}
 {
  id=ParamCrossFade_1
  children [Branch_3,Random_9,Random_13]
  properties {
   paramName=FatigueLevel
   crossFadeInput [
    {fadeInDistanceEnd=0, fadeInDistanceStart=0, fadeOutDistanceEnd=0.85, fadeOutDistanceStart=0.82, volume=1.8}
    {fadeInDistanceEnd=0.85, fadeInDistanceStart=0.82, fadeOutDistanceEnd=0.96, fadeOutDistanceStart=0.95, volume=2}
    {fadeInDistanceEnd=0.96, fadeInDistanceStart=0.95, fadeOutDistanceEnd=1, fadeOutDistanceStart=1, volume=2}
   ]
  }
 }
 {
  id=ParamCrossFade_4
  children [Branch_4,Random_17,Random_16]
  properties {
   paramName=FatigueLevel
   crossFadeInput [
    {fadeInDistanceEnd=0, fadeInDistanceStart=0, fadeOutDistanceEnd=0.85, fadeOutDistanceStart=0.82, volume=1.8}
    {fadeInDistanceEnd=0.85, fadeInDistanceStart=0.82, fadeOutDistanceEnd=0.96, fadeOutDistanceStart=0.95, volume=2}
    {fadeInDistanceEnd=0.96, fadeInDistanceStart=0.95, fadeOutDistanceEnd=1, fadeOutDistanceStart=1, volume=2}
   ]
  }
 }
 {id=QualityLevel_0, children [ParamCrossFade_1, ParamCrossFade_1, ParamCrossFade_4, ParamCrossFade_4]}
 {
  id=Random_0
  children [
   WavePlayer_19
   WavePlayer_20
   WavePlayer_25
   WavePlayer_26
   WavePlayer_36
   WavePlayer_40
   WavePlayer_41
   WavePlayer_42
   null
   null
   null
   null
   null
   null
  ]
  properties {bShouldExcludeFromBranchCulling=true, weights [1,1,1,1,1,1,1,1,1,1,1,1,1,1]}
 }
 {
  id=Random_1
  children [WavePlayer_0, WavePlayer_7, WavePlayer_8, WavePlayer_9, WavePlayer_10, WavePlayer_11]
  properties {bShouldExcludeFromBranchCulling=true, weights [1,1,1,1,1,1]}
 }
 {
  id=Random_10
  children [Random_8,Random_7]
  properties {bRandomizeWithoutReplacement=false, bShouldExcludeFromBranchCulling=true, weights [0.5,1]}
 }
 {
  id=Random_12
  children [WavePlayer_1,WavePlayer_2,WavePlayer_3,WavePlayer_4]
  properties {bShouldExcludeFromBranchCulling=true, weights [1,1,1,1]}
 }
 {id=Random_13, children [Modulator_4,Random_12,null], properties {bRandomizeWithoutReplacement=false, weights [1,0.8,0.7]}}
 {
  id=Random_14
  children [WavePlayer_21, WavePlayer_22, WavePlayer_23, WavePlayer_24]
  properties {bShouldExcludeFromBranchCulling=true, weights [1,1,1,1]}
 }
 {
  id=Random_15
  children [WavePlayer_15, WavePlayer_16, WavePlayer_17, WavePlayer_18]
  properties {bShouldExcludeFromBranchCulling=true, weights [1,1,1,1]}
 }
 {id=Random_16, children [Modulator_5,Random_15,null], properties {bRandomizeWithoutReplacement=false, weights [1,0.8,0.7]}}
 {
  id=Random_17
  children [WavePlayer_48, WavePlayer_53, WavePlayer_54, WavePlayer_55, null, null, null, null, null]
  properties {bShouldExcludeFromBranchCulling=true, weights [1,1,1,1,1,1,1,1,1]}
 }
 {id=Random_2, children [Random_0,Random_1], properties {bRandomizeWithoutReplacement=false, weights [0.5,1]}}
 {
  id=Random_5
  children [WavePlayer_5, WavePlayer_6, WavePlayer_12, WavePlayer_13, WavePlayer_14]
  properties {bShouldExcludeFromBranchCulling=true, weights [1,1,1,1,1]}
 }
 {
  id=Random_7
  children [WavePlayer_43, WavePlayer_44, WavePlayer_45, WavePlayer_46]
  properties {bShouldExcludeFromBranchCulling=true, weights [1,1,1,1]}
 }
 {
  id=Random_8
  children [WavePlayer_49, WavePlayer_50, WavePlayer_51, WavePlayer_52, null, null, null, null]
  properties {bShouldExcludeFromBranchCulling=true, weights [1,1,1,1,1,1,1,1]}
 }
 {
  id=Random_9
  children [WavePlayer_29, WavePlayer_30, WavePlayer_31, WavePlayer_33, WavePlayer_34, WavePlayer_35, null, null, null, null, null]
  properties {bShouldExcludeFromBranchCulling=true, weights [1,1,1,1,1,1,1,1,1,1,1]}
 }
 {
  id=SoundClass_0
  children [Attenuation_1]
  properties {soundClassOverride=&Fort_Creatures_Vocalization_Tactical}
 }
 {
  id=SoundClass_2
  children [Attenuation_0]
  properties {soundClassOverride=&Fort_Player_Foley_3P_Enemy_Athena}
 }
 {
  id=SoundClass_3
  children [Attenuation_3]
  properties {soundClassOverride=&Fort_Creatures_Vocalization_Tactical}
 }
 {
  id=SoundClass_4
  children [Attenuation_4]
  properties {soundClassOverride=&Fort_Creatures_Vocalization_Tactical}
 }
 {id=WavePlayer_0, properties {soundWaveAssetPtr=&/Burt_Vocal_Chase_Breathing_01}}
 {id=WavePlayer_1, properties {soundWaveAssetPtr=&/Burt_Vocal_Agitated_01}}
 {id=WavePlayer_10, properties {soundWaveAssetPtr=&/Burt_Vocal_Chase_Breathing_05}}
 {id=WavePlayer_11, properties {soundWaveAssetPtr=&/Burt_Vocal_Chase_Breathing_06}}
 {id=WavePlayer_12, properties {soundWaveAssetPtr=&/Burt_Vocal_Sprint_Tiring_Breathing_03}}
 {id=WavePlayer_13, properties {soundWaveAssetPtr=&/Burt_Vocal_Sprint_Tiring_Breathing_05}}
 {id=WavePlayer_14, properties {soundWaveAssetPtr=&/Burt_Vocal_Sprint_Tiring_Breathing_06}}
 {id=WavePlayer_15, properties {soundWaveAssetPtr=&Burt_Vocal_Agitated_01}}
 {id=WavePlayer_16, properties {soundWaveAssetPtr=&/Burt_Vocal_Agitated_02}}
 {id=WavePlayer_17, properties {soundWaveAssetPtr=&/Burt_Vocal_Agitated_06}}
 {id=WavePlayer_18, properties {soundWaveAssetPtr=&/Burt_Vocal_Agitated_07}}
 {id=WavePlayer_19, properties {soundWaveAssetPtr=&/Burt_Vocal_ChargeAttack_01}}
 {id=WavePlayer_2, properties {soundWaveAssetPtr=&Burt_Vocal_Agitated_02}}
 {id=WavePlayer_20, properties {soundWaveAssetPtr=&/Burt_Vocal_ChargeAttack_02}}
 {id=WavePlayer_21, properties {soundWaveAssetPtr=&/Burt_Vocal_Sprint_Tiring_Breathing_01}}
 {id=WavePlayer_22, properties {soundWaveAssetPtr=&/Burt_Vocal_Sprint_Tiring_Breathing_02}}
 {id=WavePlayer_23, properties {soundWaveAssetPtr=&Burt_Vocal_Sprint_Tiring_Breathing_03}}
 {id=WavePlayer_24, properties {soundWaveAssetPtr=&Burt_Vocal_Sprint_Tiring_Breathing_05}}
 {id=WavePlayer_25, properties {soundWaveAssetPtr=&/Burt_Vocal_ChargeAttack_03}}
 {id=WavePlayer_26, properties {soundWaveAssetPtr=&/Burt_Vocal_ChargeAttack_04}}
 {id=WavePlayer_29, properties {soundWaveAssetPtr=&Burt_Vocal_Sprint_Tiring_Breathing_01}}
 {id=WavePlayer_3, properties {soundWaveAssetPtr=&Burt_Vocal_Agitated_06}}
 {id=WavePlayer_30, properties {soundWaveAssetPtr=&Burt_Vocal_Sprint_Tiring_Breathing_02}}
 {id=WavePlayer_31, properties {soundWaveAssetPtr=&Burt_Vocal_Sprint_Tiring_Breathing_03}}
 {id=WavePlayer_33, properties {soundWaveAssetPtr=&Burt_Vocal_Sprint_Tiring_Breathing_05}}
 {id=WavePlayer_34, properties {soundWaveAssetPtr=&Burt_Vocal_Sprint_Tiring_Breathing_06}}
 {id=WavePlayer_35, properties {soundWaveAssetPtr=&/Burt_Vocal_Sprint_Tiring_Breathing_07}}
 {id=WavePlayer_36, properties {soundWaveAssetPtr=&/Burt_Vocal_ChargeAttack_05}}
 {id=WavePlayer_4, properties {soundWaveAssetPtr=&Burt_Vocal_Agitated_07}}
 {id=WavePlayer_40, properties {soundWaveAssetPtr=&/Burt_Vocal_ChargeAttack_06}}
 {id=WavePlayer_41, properties {soundWaveAssetPtr=&/Burt_Vocal_ChargeAttack_07}}
 {id=WavePlayer_42, properties {soundWaveAssetPtr=&/Burt_Vocal_ChargeAttack_08}}
 {id=WavePlayer_43, properties {soundWaveAssetPtr=&Burt_Vocal_Chase_Breathing_01}}
 {id=WavePlayer_44, properties {soundWaveAssetPtr=&/Burt_Vocal_Chase_Breathing_09}}
 {id=WavePlayer_45, properties {soundWaveAssetPtr=&/Burt_Vocal_Chase_Breathing_03}}
 {id=WavePlayer_46, properties {soundWaveAssetPtr=&/Burt_Vocal_Chase_Breathing_04}}
 {id=WavePlayer_48, properties {soundWaveAssetPtr=&Burt_Vocal_Sprint_Tiring_Breathing_01}}
 {id=WavePlayer_49, properties {soundWaveAssetPtr=&Burt_Vocal_ChargeAttack_01}}
 {id=WavePlayer_5, properties {soundWaveAssetPtr=&Burt_Vocal_Sprint_Tiring_Breathing_01}}
 {id=WavePlayer_50, properties {soundWaveAssetPtr=&Burt_Vocal_ChargeAttack_02}}
 {id=WavePlayer_51, properties {soundWaveAssetPtr=&Burt_Vocal_ChargeAttack_03}}
 {id=WavePlayer_52, properties {soundWaveAssetPtr=&Burt_Vocal_ChargeAttack_04}}
 {id=WavePlayer_53, properties {soundWaveAssetPtr=&Burt_Vocal_Sprint_Tiring_Breathing_02}}
 {id=WavePlayer_54, properties {soundWaveAssetPtr=&Burt_Vocal_Sprint_Tiring_Breathing_03}}
 {id=WavePlayer_55, properties {soundWaveAssetPtr=&Burt_Vocal_Sprint_Tiring_Breathing_05}}
 {id=WavePlayer_6, properties {soundWaveAssetPtr=&Burt_Vocal_Sprint_Tiring_Breathing_02}}
 {id=WavePlayer_7, properties {soundWaveAssetPtr=&Burt_Vocal_Chase_Breathing_09}}
 {id=WavePlayer_8, properties {soundWaveAssetPtr=&Burt_Vocal_Chase_Breathing_03}}
 {id=WavePlayer_9, properties {soundWaveAssetPtr=&Burt_Vocal_Chase_Breathing_04}}
]
properties {
 maxDistance=9000
 pitchMultiplier=0.9
 volumeMultiplier=1.25
 concurrencySet [
  /Game/Sounds/Concurrency_Rules/Burt_Vocal_Conc
  &Athena_Footstep_Global_Conc
  /Game/Sounds/Concurrency_Rules/LimitToOnePreventNewQuick
 ]
}
