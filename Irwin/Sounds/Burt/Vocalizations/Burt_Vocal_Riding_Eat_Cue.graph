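# Sound cue for Burt's eating vocalization while ridden, varied by listener relationship to the rider.
# IsLocal and IsRiddenByEnemy branches select sound class and attenuation, with enemy-ridden using its own riding attenuation and non-enemy playing quieter.
# All paths share random pitch and volume modulation into a quality switch between two mixers.
# Each mixer layers a random short eat clip over a slightly delayed, quieter random eating clip.
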SoundCue /Irwin/Sounds/Burt/Vocalizations/Burt_Vocal_Riding_Eat_Cue
root=Branch_2
nodes [
 {id=Attenuation_0, children [Modulator_1], properties {attenuationSettings=/Irwin/Sounds/Settings/Prey_Vocalization_Att}}
 {id=Attenuation_2, children [Modulator_1], properties {attenuationSettings=/Game/Sounds/Attenuation/Fort_Riding_Enemy_VO_Att}}
 {id=Attenuation_3, children [Modulator_1], properties {attenuationSettings=&Prey_Vocalization_Att}}
 {id=Attenuation_4, children [Modulator_1], properties {attenuationSettings=&Prey_Vocalization_Att}}
 {id=Branch_1, children [Modulator_3,Modulator_2,SoundClass_0], properties {boolParameterName=IsRiddenByEnemy}}
 {id=Branch_2, children [SoundClass_4,Branch_1,Branch_1], properties {boolParameterName=IsLocal}}
 {id=Delay_0, children [Random_1], properties {delayMax=0.1, delayMin=0.1}}
 {id=Delay_1, children [Random_2], properties {delayMax=0.1, delayMin=0.1}}
 {id=Mixer_1, children [Random_0,Delay_0], properties {inputVolume [2,1]}}
 {id=Mixer_2, children [Random_4,Delay_1,null], properties {inputVolume [2,1,1]}}
 {id=Modulator_1, children [QualityLevel_0], properties {pitchMax=1.1, pitchMin=0.9, volumeMax=1, volumeMin=0.9}}
 {id=Modulator_2, children [SoundClass_3], properties {pitchMax=1, pitchMin=1, volumeMax=0.3, volumeMin=0.3}}
 {id=Modulator_3, children [SoundClass_2], properties {pitchMax=1, pitchMin=1, volumeMax=0.7, volumeMin=0.7}}
 {id=QualityLevel_0, children [Mixer_1,Mixer_1,Mixer_2,Mixer_2]}
 {
  id=Random_0
  children [WavePlayer_6, WavePlayer_7, WavePlayer_8, WavePlayer_9, WavePlayer_10, WavePlayer_11]
  properties {weights [1,1,1,1,1,1]}
 }
 {id=Random_1, children [WavePlayer_0,WavePlayer_1,WavePlayer_2,WavePlayer_3], properties {weights [1,1,1,1]}}
 {
  id=Random_2
  children [WavePlayer_4, WavePlayer_5, WavePlayer_12, WavePlayer_13]
  properties {bShouldExcludeFromBranchCulling=true, weights [1,1,1,1]}
 }
 {
  id=Random_4
  children [WavePlayer_29, WavePlayer_32, WavePlayer_33, WavePlayer_34]
  properties {bShouldExcludeFromBranchCulling=true, weights [1,1,1,1]}
 }
 {
  id=SoundClass_0
  children [Attenuation_0]
  properties {soundClassOverride=&Fort_Creatures_Vocalization_Tactical}
 }
 {
  id=SoundClass_2
  children [Attenuation_2]
  properties {soundClassOverride=&Fort_Player_Foley_3P_Enemy_Athena}
 }
 {
  id=SoundClass_3
  children [Attenuation_3]
  properties {soundClassOverride=&Fort_Creatures_Vocalization_Tactical}
 }
 {
  id=SoundClass_4
  children [Attenuation_4]
  properties {soundClassOverride=&Fort_Creatures_Vocalization_Tactical}
 }
 {id=WavePlayer_0, properties {soundWaveAssetPtr=&/Burt_Vocal_Eating_04}}
 {id=WavePlayer_1, properties {soundWaveAssetPtr=&/Burt_Vocal_Eating_01}}
 {id=WavePlayer_10, properties {soundWaveAssetPtr=&/Burt_Vocal_Eat_Short_05}}
 {id=WavePlayer_11, properties {soundWaveAssetPtr=&/Burt_Vocal_Eat_Short_06}}
 {id=WavePlayer_12, properties {soundWaveAssetPtr=&/Burt_Vocal_Eating_02}}
 {id=WavePlayer_13, properties {soundWaveAssetPtr=&/Burt_Vocal_Eating_03}}
 {id=WavePlayer_2, properties {soundWaveAssetPtr=&Burt_Vocal_Eating_02}}
 {id=WavePlayer_29, properties {soundWaveAssetPtr=&/Burt_Vocal_Eat_Short_01}}
 {id=WavePlayer_3, properties {soundWaveAssetPtr=&Burt_Vocal_Eating_03}}
 {id=WavePlayer_32, properties {soundWaveAssetPtr=&/Burt_Vocal_Eat_Short_02}}
 {id=WavePlayer_33, properties {soundWaveAssetPtr=&/Burt_Vocal_Eat_Short_03}}
 {id=WavePlayer_34, properties {soundWaveAssetPtr=&/Burt_Vocal_Eat_Short_04}}
 {id=WavePlayer_4, properties {soundWaveAssetPtr=&Burt_Vocal_Eating_04}}
 {id=WavePlayer_5, properties {soundWaveAssetPtr=&Burt_Vocal_Eating_01}}
 {id=WavePlayer_6, properties {soundWaveAssetPtr=&Burt_Vocal_Eat_Short_01}}
 {id=WavePlayer_7, properties {soundWaveAssetPtr=&Burt_Vocal_Eat_Short_02}}
 {id=WavePlayer_8, properties {soundWaveAssetPtr=&Burt_Vocal_Eat_Short_03}}
 {id=WavePlayer_9, properties {soundWaveAssetPtr=&Burt_Vocal_Eat_Short_04}}
]
properties {
 attenuationSettings=&Prey_Vocalization_Att
 maxDistance=9000
 pitchMultiplier=1.1
 soundClassObject=/Game/Sounds/FortSoundClassesAndModes/Fort_Creatures_Vocalization_Normal
 volumeMultiplier=1.9
 concurrencySet [&Athena_Footstep_Local_Conc, &Athena_Footstep_Global_Conc]
}
