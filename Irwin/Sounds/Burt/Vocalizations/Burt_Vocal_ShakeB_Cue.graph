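# Sound cue for Burt's "ShakeB" creature vocalization, using prey vocalization attenuation and the tactical creature vocalization sound class.
# A modulator varies pitch and volume over a quality-level switch.
# The first two quality slots pick randomly from four waves; the third uses a two-wave random, exempt from branch culling, and the fourth is empty.
# Concurrency borrows the Athena local and global footstep settings rather than vocalization-specific ones.
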SoundCue /Irwin/Sounds/Burt/Vocalizations/Burt_Vocal_ShakeB_Cue
root=Modulator_0
nodes [
 {id=Modulator_0, children [QualityLevel_0], properties {pitchMax=1.1, volumeMax=1.1, volumeMin=0.9}}
 {id=QualityLevel_0, children [Random_1,Random_1,Random_0,null]}
 {id=Random_0, children [WavePlayer_1,WavePlayer_2], properties {bShouldExcludeFromBranchCulling=true, weights [1,1]}}
 {id=Random_1, children [WavePlayer_0,WavePlayer_5,WavePlayer_6,WavePlayer_7], properties {weights [1,1,1,1]}}
 {id=WavePlayer_0, properties {soundWaveAssetPtr=&/Burt_Vocal_ShakeB_01}}
 {id=WavePlayer_1, properties {soundWaveAssetPtr=&Burt_Vocal_ShakeB_01}}
 {id=WavePlayer_2, properties {soundWaveAssetPtr=&/Burt_Vocal_ShakeB_02}}
 {id=WavePlayer_5, properties {soundWaveAssetPtr=&Burt_Vocal_ShakeB_02}}
 {id=WavePlayer_6, properties {soundWaveAssetPtr=&/Burt_Vocal_ShakeB_03}}
 {id=WavePlayer_7, properties {soundWaveAssetPtr=&/Burt_Vocal_ShakeB_04}}
]
properties {
 attenuationSettings=/Irwin/Sounds/Settings/Prey_Vocalization_Att
 maxDistance=6350
 soundClassObject=&Fort_Creatures_Vocalization_Tactical
 volumeMultiplier=1.4
 concurrencySet [&Athena_Footstep_Local_Conc, &Athena_Footstep_Global_Conc]
}
