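# Sound cue for Burt's short snort vocalization, classed as a creature vocalization with prey attenuation.
# A quality switch picks a six-clip random pool for the first two levels and a four-clip subset for the last two, drawing on Phrase, Recover and Tame recordings.
# A modulator varies pitch and volume on the result.
# Concurrency combines a one-per-owner voice limit with shared Athena local and global footstep limits.
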SoundCue /Irwin/Sounds/Burt/Vocalizations/Burt_Vocal_SnortShort_Cue
root=Modulator_0
nodes [
 {id=Modulator_0, children [QualityLevel_0], properties {pitchMax=1.1, volumeMax=1.1, volumeMin=0.9}}
 {id=QualityLevel_0, children [Random_1,Random_1,Random_0,Random_0]}
 {
  id=Random_0
  children [WavePlayer_2,WavePlayer_1,WavePlayer_5,WavePlayer_0]
  properties {bShouldExcludeFromBranchCulling=true, weights [1,1,1,1]}
 }
 {
  id=Random_1
  children [WavePlayer_13, WavePlayer_14, WavePlayer_15, WavePlayer_4, WavePlayer_16, WavePlayer_12]
  properties {weights [1,1,1,1,1,1]}
 }
 {id=WavePlayer_0, properties {soundWaveAssetPtr=&/Burt_Vocal_Tame_04}}
 {id=WavePlayer_1, properties {soundWaveAssetPtr=&/Burt_Vocal_Phrase_02}}
 {id=WavePlayer_12, properties {soundWaveAssetPtr=&Burt_Vocal_Phrase_02}}
 {id=WavePlayer_13, properties {soundWaveAssetPtr=&/Burt_Vocal_Phrase_04}}
 {id=WavePlayer_14, properties {soundWaveAssetPtr=&/Burt_Vocal_Phrase_05}}
 {id=WavePlayer_15, properties {soundWaveAssetPtr=&/Burt_Vocal_Recover_01}}
 {id=WavePlayer_16, properties {soundWaveAssetPtr=&/Burt_Vocal_Recover_02}}
 {id=WavePlayer_2, properties {soundWaveAssetPtr=&Burt_Vocal_Phrase_04}}
 {id=WavePlayer_4, properties {soundWaveAssetPtr=&Burt_Vocal_Tame_04}}
 {id=WavePlayer_5, properties {soundWaveAssetPtr=&Burt_Vocal_Recover_01}}
]
properties {
 attenuationSettings=/Irwin/Sounds/Settings/Prey_Vocalization_Att
 maxDistance=6350
 soundClassObject=/Game/Sounds/FortSoundClassesAndModes/Fort_Creatures_Vocalization_Normal
 volumeMultiplier=1.4
 concurrencySet [
  &Athena_Footstep_Local_Conc
  &Athena_Footstep_Global_Conc
  /Irwin/Sounds/Settings/Conc/SetOneLimitToOwner_VO_Conc
 ]
}
